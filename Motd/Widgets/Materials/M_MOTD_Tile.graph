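# Translucent UI material for a message-of-the-day (MOTD) tile, showing rounded-corner thumbnail art with an animated loading state.
# Image_A and Image_B are scaled to fill with independent zoom, crossfaded by Progress, then tinted by an eased vertical GradientColor overlay whose strength follows Anim_Gradient.
# A time-driven, eased linear-gradient sweep forms the loading mask, which Display Loading gates.
# Named reroutes feed lerps that blend in flat grey and reduced rounded-mask opacity.
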Material /Motd/Widgets/Materials/M_MOTD_Tile
graph {
 properties {blendMode=BLEND_Translucent, materialDomain=MD_UI, pixelDepthOffsetMode=PDOM_Legacy, usedShadingModels=Default Lit}
 nodes [
  LinearInterpolate(A=NamedRerouteUsage(), B=Constant3Vector() {constant=<rgba 0.3515 0.3515 0.3515 1>}, Alpha=NamedRerouteUsage())
  n1 = MaterialFunctionCall(Corner Radius=Constant() {r=20}, Radius in Pixels=Constant() {r=1}) {materialFunction=&MF_UI_CoreUI_RoundedMask}
  n2 = Multiply(
   A=NamedRerouteDeclaration(None=ScalarParameter() {defaultValue=1, parameterName=Display Loading, sliderMax=1}) {name=Is Displaying Loading, nodeColor=<rgba 0 0 0 1>}
   B=NamedRerouteUsage()
  )
  LinearInterpolate(A=@n1.Shape Mask, B=Multiply(A=LinearInterpolate(Alpha=@n2) {constA=0.2, constB=0.4}, B=@n1.Shape Mask), Alpha=NamedRerouteUsage())
  n3 = MaterialFunctionCall() {materialFunction=&GetUserInterfaceUV}
  n4 = TextureObjectParameter() {group=Image, parameterName=Image_A}
  n5 = TextureProperty(Tex=@n4)
  n6 = TextureSample(
   UVs=MaterialFunctionCall(
    UVs=MaterialFunctionCall(UV=@n3.Normalized UV, TextureSize=@n5) {materialFunction=&MF_UI_UVScaling}.Scaled To Fill
    Texture Scale=ScalarParameter() {defaultValue=1, parameterName=Scale_A}
   ) {materialFunction=&ScaleUVsByCenter}.UVs
   Tex=@n4
  ) {samplerSource=SSM_Clamp_WorldGroupSettings}
  n7 = TextureObjectParameter() {group=Image, parameterName=Image_B}
  n8 = TextureProperty(Tex=@n7)
  n9 = TextureSample(
   UVs=MaterialFunctionCall(
    UVs=MaterialFunctionCall(UV=@n3.Normalized UV, TextureSize=@n8) {materialFunction=&MF_UI_UVScaling}.Scaled To Fill
    Texture Scale=ScalarParameter() {defaultValue=1, parameterName=Scale_B}
   ) {materialFunction=&ScaleUVsByCenter}.UVs
   Tex=@n7
  ) {samplerSource=SSM_Clamp_WorldGroupSettings}
  n10 = ScalarParameter() {parameterName=Progress, sliderMax=1}
  n11 = MaterialFunctionCall("Value (0-1)"=ComponentMask(None=MaterialFunctionCall() {materialFunction=&GetUserInterfaceUV}.Normalized UV) {g=true}) {materialFunction=&MF_UI_EaseCurves}
  n12 = LinearInterpolate(Alpha=ScalarParameter() {parameterName=Anim_Gradient, sliderMax=1}) {constA=0.6, constB=0.9}
  NamedRerouteDeclaration(None=LinearInterpolate(
    A=LinearInterpolate(A=@n6.RGB, B=@n9.RGB, Alpha=@n10)
    B=VectorParameter() {group=Gradient Settings, parameterName=GradientColor, defaultValue=<rgba 0.0012 0.0137 0.0865 1>}.RGB
    Alpha=Multiply(A=@n11.EaseInOutSine, B=@n12)
   )) {name=MOTD Thumbnails, nodeColor=<rgba 0 0 0 1>}
  n13 = Constant() {r=2}
  n14 = Constant() {r=0.125}
  n15 = MaterialFunctionCall(
   GradientEnd=MaterialFunctionCall("Value (0-1)"=MaterialFunctionCall!(Animation Duration=@n13, Delay=@n14, Reset to Default) {materialFunction=/FortUILibrary/MaterialFunctions/Time/MF_UI_AnimationDelay}.Result) {materialFunction=&MF_UI_EaseCurves}.EaseInOutSine
   RotationAngle=Constant()
  ) {materialFunction=&MF_UI_LinearGradient}
  n16 = MaterialFunctionCall!(A=@n15.Result, B=Constant() {r=1}) {materialFunction=/FortUILibrary/MaterialFunctions/Basics/MF_UI_BooleanOperations}
  NamedRerouteDeclaration(None=Multiply(A=@n15.Result, B=OneMinus(None=@n16.IsEqual))) {name=Loading Mask, nodeColor=<rgba 0 0 0 1>}
 ]
}
