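# WBP_ArrowButton is a UIKit generic button subclass that shows an arrow whose direction can be flipped.
# PreConstruct sets the outline material's corner radius parameters, then applies the InverseArrow flag.
# InverseArrowDirection stores the flag, switches the input prompt's alignment and nudge to the matching side, and refreshes it.
# It then flips the arrow through the icon material's Rotation parameter, so no separate mirrored widget logic is needed.
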WidgetBlueprint /Motd/Widgets/WBP_ArrowButton
graphs [
 EventGraph {nodes [
   n1 = Event PreConstruct()
   > Parent: PreConstruct(IsDesignTime=@n1.IsDesignTime)
   > SetCornerRadius()
   > InverseArrowDirection!(execute, self, InverseArrow:bool=Get InverseArrow() -> (InverseArrow:bool)) -> (then)
   Event Construct()
   > Parent: Construct()
  ]}
 InverseArrowDirection {nodes [
   n1 = InverseArrowDirection!() -> (then, InverseArrow:bool=~) {is=title K2Node_FunctionEntry}
   > Set InverseArrow(InverseArrow:bool=@n1) -> (Output_Get:bool)
   n2 = Get InverseArrow()
   > Set PromptHorizontalAlignment(PromptHorizontalAlignment:EHorizontalAlignment Enum=Select(Option 0:EHorizontalAlignment Enum=HAlign_Right, Option 1:EHorizontalAlignment Enum=HAlign_Left, Index=@n2) -> (ReturnValue:EHorizontalAlignment Enum)) -> (Output_Get:EHorizontalAlignment Enum=HAlign_Fill)
   > Set PromptNudge(PromptNudge:Vector 2D Structure=Select(Option 0:Vector 2D Structure=(X=48,Y=0), Option 1:Vector 2D Structure=(X=-48,Y=0), Index=@n2) -> (ReturnValue:Vector 2D Structure)) -> (Output_Get:Vector 2D Structure)
   > UpdateInputPrompt(self:WBP UIKit Button Generic Base Object Reference)
   > n3 = Cast To WBP_UIKit_Block_Image!(execute, Object=Get Block_Icon() -> (Block_Icon:WBP UIKit Block Image Object Reference)) -> (then, CastFailed, AsWBP UIKit Block Image:WBP UIKit Block Image Object Reference) {is=title K2Node_DynamicCast}
   > SetScalarParameterValue(
    self=Get ImageMID!(self:WBP UIKit Block Image Object Reference=@n3) -> (ImageMID:Material Instance Dynamic Object Reference)
    ParameterName=Rotation
    Value=Select(Option 0:float=0, Option 1:float=0.5, Index=Get InverseArrow()) -> (ReturnValue:float)
   )
  ]}
 SetCornerRadius {nodes [
   SetCornerRadius() {is=title K2Node_FunctionEntry}
   > n1 = Cast To WBP_UIKit_Block_BackgroundWithOutline!(execute, Object=Get Block_Background() -> (Block_Background:WBP UIKit Block Background Base Object Reference)) -> (then, CastFailed, AsWBP UIKit Block Background with Outline:WBP UIKit Block Background With Outline Object Reference) {is=title K2Node_DynamicCast}
   n2 = Get OutlineDynamicMaterial!(self:WBP UIKit Block Background Base Object Reference=@n1) -> (OutlineDynamicMaterial:Material Instance Dynamic Object Reference)
   > SetScalarParameterValue(self=@n2, ParameterName=Radius in Pixels, Value=0)
   > SetVectorParameterValue!(execute, self=@n2, ParameterName=Corner Radius, Value_R:float=0.325, Value_G:float=0.325, Value_B:float=0.325, Value_A:float=0.325) -> (then)
  ]}
]
variables [InverseArrow {type=boolean}]
widgets {
 info {
  inheritedWidgetCount=0
  parentClass=/FortUILibrary/UIKit/Buttons/WBP_UIKit_Button_Generic.WBP_UIKit_Button_Generic_C
  rootWidgetClass=null
  widgetCount=0
 }
 widgets []
}
