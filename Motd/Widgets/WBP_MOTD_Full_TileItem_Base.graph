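# Base widget for fullscreen MOTD tiles, extending FortDynamicMotdFullscreenTile with a single named Content slot for derived layouts.
# PreConstruct caches and broadcasts the compact UI flag, then in compact mode sets a taller compact size for 4:3 or narrower screens.
# CF_ helper functions select compact or default size and margin.
# SwitchPage maps left/previous and right/next navigation to event dispatchers, while SetActive and Set Vertical Layout are empty stubs.
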WidgetBlueprint /Motd/Widgets/WBP_MOTD_Full_TileItem_Base
graphs [
 CF_CompactUI_DesiredMargin {nodes [
   n1 = CF_CompactUI_DesiredMargin!() -> (then, Margin_CompactUI:Margin Structure, Margin_Default:Margin Structure, IsUsingCompactUI:bool=~) {is=title K2Node_FunctionEntry}
   > n2 = Branch(Condition=@n1.IsUsingCompactUI)
   > Return Node!(execute, Margin_Desired:Margin Structure=Get Margin_CompactUI() -> (Margin_CompactUI:Margin Structure))
   Return Node(execute=@n2.else, Margin_Desired=Get Margin_Default() -> (Margin_Default:Margin Structure))
  ]}
 CF_CompactUI_DesiredSize {nodes [
   n1 = CF_CompactUI_DesiredSize!() -> (then, Size_CompactUI:Vector 2D Structure, Size_Default:Vector 2D Structure, IsUsingCompactUI:bool=~) {is=title K2Node_FunctionEntry}
   > n2 = Branch(Condition=@n1.IsUsingCompactUI)
   > Return Node!(execute, Size_Desired:Vector 2D Structure=Get Size_CompactUI() -> (Size_CompactUI:Vector 2D Structure))
   Return Node(execute=@n2.else, Size_Desired=Get Size_Default() -> (Size_Default:Vector 2D Structure))
  ]}
 CloseWidget {nodes [CloseWidget() {is=title K2Node_FunctionEntry}]}
 EventGraph {nodes [
   Event PreConstruct()
   > n1 = Set with Broadcast IsUsingCompactUI!(execute, IsUsingCompactUI:bool=IsUsingCompactUI()) -> (then, Output_Get:bool)
   > Branch(Condition=@n1)
   > Set Size_CompactUI(Size_CompactUI:Vector 2D Structure=Select(
     Option 0:Vector 2D Structure=(X=1648,Y=700)
     Option 1:Vector 2D Structure=(X=1648,Y=1150)
     Index=CompareToScreenAspectRatio!(Comparison:EAspectRatioComparison Enum=LessThanOrEqualTo, AspectRatio:EAspectRatio Enum=Ratio_4_3) -> (ReturnValue:bool)
    ) -> (ReturnValue:Vector 2D Structure)) -> (Output_Get:Vector 2D Structure)
  ]}
 NavigateLeft {nodes [NavigateLeft() {is=title K2Node_FunctionEntry}]}
 NavigateRight {nodes [NavigateRight() {is=title K2Node_FunctionEntry}]}
 Set Vertical Layout {nodes [Set Vertical Layout!() -> (then, IsVertical:bool=~) {is=title K2Node_FunctionEntry}]}
 SetActive {nodes [SetActive!() -> (then, Active:bool=~, Scroll:bool=~) {is=title K2Node_FunctionEntry}]}
 SwitchPage {nodes [
   n1 = SwitchPage!() -> (then, Navigation:EUINavigation Enum) {is=title K2Node_FunctionEntry}
   > n2 = Switch on EUINavigation!(execute, Selection:EUINavigation Enum=@n1) -> (Left:Exec, Right:Exec, Up:Exec, Down:Exec, Next:Exec, Previous:Exec, Invalid:Exec) {is=title K2Node_SwitchEnum}
   n3 = Call Navigate Left!(execute:Exec=[@n2.Left, @n2.Previous], self:WBP MOTD Full Tile Item Base Object Reference) -> (then:Exec) {is=title K2Node_CallDelegate}
   Return Node!(execute=[@n4.then, @n3.then], Widget:Widget Object Reference)
   n4 = Call Navigate Right!(execute:Exec=[@n2.Next, @n2.Right], self:WBP MOTD Full Tile Item Base Object Reference) -> (then:Exec) {is=title K2Node_CallDelegate}
  ]}
]
variables [
 CloseWidget {}
 IsUsingCompactUI {type=boolean}
 NavigateLeft {}
 NavigateRight {}
 Size_CompactUI {type=Vector2D, default=<1800 700>}
 Size_Default {type=Vector2D, default=<1648 920>}
]
widgets {
 info {inheritedWidgetCount=0, parentClass=/Script/FortniteUI.FortDynamicMotdFullscreenTile, rootWidgetClass=&NamedSlot, widgetCount=1}
 widgets [Content {widgetClassPath=&NamedSlot, bIsVariable=true, widget=Content}]
}
