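# Full-size MOTD tile item that shows a community post by wrapping WBP_CommunityPost_Content in the shared full-tile base.
# On Construct it queries details from the MOTD view model's CommunityAnnouncement; activation is forwarded to the content, which also supplies the focus target.
# Content navigation switches pages, and a successful query refreshes focus.
# A decoy button bound to the ReportPost input reports the post and closes, as does the content's CTA button.
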WidgetBlueprint /Motd/Widgets/WBP_MOTD_CommunityPost_Full_TileItem
graphs [
 BP_GetDesiredFocusTarget {nodes [
   Get Desired Focus Target()
   > Return Node!(
    execute
    ReturnValue:Widget Object Reference=GetFocusTarget!(self:WBP Community Post Content Object Reference=Get CommunityPost_Content() -> (CommunityPost_Content:WBP Community Post Content Object Reference)) -> (TargetWidget:Widget Object Reference)
   )
  ]}
 EventGraph {nodes [
   "On Clicked (DecoyButton_Report)"!() -> (OutputDelegate, then, Button:Common Button Base Object Reference) {is=title K2Node_ComponentBoundEvent}
   > Report(self:WBP Community Post Content Object Reference=Get CommunityPost_Content())
   > Call Close Widget!(execute:Exec, self:WBP MOTD Full Tile Item Base Object Reference) -> (then:Exec) {is=title K2Node_CallDelegate}
   "On CTAButton Clicked (CommunityPost_Content)"() {is=title K2Node_ComponentBoundEvent}
   > Call Close Widget()
   n1 = "On Navigate (CommunityPost_Content)"!() -> (OutputDelegate, then, Navigation:EUINavigation Enum) {is=title K2Node_ComponentBoundEvent}
   > SwitchPage!(execute, self:WBP MOTD Full Tile Item Base Object Reference, Navigation:EUINavigation Enum=@n1.Navigation) -> (then, Widget:Widget Object Reference)
   "On Success Finished Querying (CommunityPost_Content)"() {is=title K2Node_ComponentBoundEvent}
   > RequestRefreshFocus(self:Common Activatable Widget Object Reference)
   Event On Activated()
   > OnActivated(self:WBP Community Post Content Object Reference=Get CommunityPost_Content())
   Event Construct()
   > QueryDetails(self:Community Announcement VM Object Reference=Get CommunityAnnouncement!(self:Motd View Model Object Reference=Get MotdViewModel() -> (MotdViewModel:Motd View Model Object Reference)) -> (CommunityAnnouncement:Community Announcement VM Object Reference))
  ]}
]
variables []
widgets {
 info {
  inheritedWidgetCount=0
  namedSlotCount=1
  parentClass=&/WBP_MOTD_Full_TileItem_Base.WBP_MOTD_Full_TileItem_Base_C
  rootWidgetClass=null
  widgetCount=5
 }
 widgets [
  Overlay_0 {widgetClassPath=&Overlay, widget=Overlay_0}
  GridPanel_0 {
   widgetClassPath=&GridPanel
   parent=Overlay_0
   widget=GridPanel_0
   properties {columnFill [1], rowFill [1]}
   slot {class=OverlaySlot, properties {horizontalAlignment=HAlign_Fill, verticalAlignment=VAlign_Fill, padding=<ltrb 20 0 20 0>}}
  }
  S_ItemSize {
   widgetClassPath=&Spacer
   bIsVariable=true
   parent=GridPanel_0
   widget=S_ItemSize
   properties {size=<1648 920>}
   slot {class=GridSlot}
  }
  CommunityPost_Content {
   widgetClassPath=/FortniteCoreUI/UI/Community/WBP_CommunityPost_Content
   bIsVariable=true
   parent=GridPanel_0
   widget=CommunityPost_Content
   slot {class=GridSlot}
  }
  DecoyButton_Report {
   widgetClassPath=/Game/UI/Frontend/BottomBar/WBP_BottomBarDecoyButton
   bIsVariable=true
   parent=Overlay_0
   widget=DecoyButton_Report
   properties {
    bDisplayInputActionWhenNotInteractable=false
    bHideInputActionWithKeyboard=true
    triggeringInputAction {dataTable=&InputActionDataTable, rowName=ReportPost}
   }
   slot {class=OverlaySlot}
  }
 ]
}
