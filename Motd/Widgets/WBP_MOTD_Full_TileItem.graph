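# Fullscreen Message of the Day tile widget showing title, body text and call-to-action buttons over an image or video background.
# Builds Loud or Regular buttons from Motd data once; clicks map child index to action, log analytics, optionally close, then run it.
# SetActive toggles content and video, delaying start after scrolling.
# CMS images and looping media-player video share one dynamic-material texture parameter; layout switches between horizontal and scrollable vertical.
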WidgetBlueprint /Motd/Widgets/WBP_MOTD_Full_TileItem
animations [Loaded {
  binding_count=1
  length_seconds=0.1
  object=&self:Loaded
  sequence {
   bindings [Image {
     id=#1
     possessed_class=Image
     tracks [{
       type=MovieSceneColorTrack
       display_name=Color and Opacity
       sections [{
         type=MovieSceneColorSection
         end_frame=6
         start_frame=0
         channels [
          Color.R {type=MovieSceneScriptingFloatChannel, const [1,1]}
          Color.G {type=MovieSceneScriptingFloatChannel, const [1,1]}
          Color.B {type=MovieSceneScriptingFloatChannel, const [1,1]}
          Color.A {type=MovieSceneScriptingFloatChannel, keys [[0,0],[6,1]]}
         ]
        }]
      }]
    }]
   display_rate {numerator=60}
  }
 }]
graphs [
 BP_GetDesiredFocusTarget {nodes [
   Get Desired Focus Target()
   > Return Node!(
    execute
    ReturnValue:Widget Object Reference=GetChildAt!(self:Panel Widget Object Reference=Get VerticalBox_Buttons() -> (VerticalBox_Buttons:Vertical Box Object Reference), Index:int=0) -> (ReturnValue:Widget Object Reference)
   )
  ]}
 Button Pressed {nodes [
   n1 = Button Pressed!() -> (then, Action:Motd Action Base Object Reference, Button Context:Common Button Base Object Reference) {is=title K2Node_FunctionEntry}
   > n2 = Cast To FortCTAButton!(execute, Object=@n1.Button Context) -> (then, CastFailed, AsFort CTAButton:Fort CTAButton Object Reference) {is=title K2Node_DynamicCast}
   > Fire Event Element Interacted(
    Element=@n2
    AdditionalInfo="Add Text (Localized)"!(
     Info:"Flow Analytics Additional Info Structure (by ref)"=Make Flow Analytics Additional Info!() -> (FlowAnalyticsAdditionalInfo:Flow Analytics Additional Info Structure) {is=title K2Node_MakeStruct}
     Key:String=Button
     Value:"Text (by ref)"=Get Text(self:Fort CTAButton Object Reference=@n2)
    ) -> (ReturnValue:Flow Analytics Additional Info Structure)
   )
   > n3 = Cast To FortPlayerController(Object=GetOwningPlayer())
   > HandleVideoAction!(execute, self, Action:Motd Action Base Object Reference=Get Action() -> (Action:Motd Action Base Object Reference)) -> (then)
   > StopVideo!(execute, self, SkipClosedControlBroadcast:bool=true) -> (then)
   > n4 = Branch(Condition=ShouldCloseWidget!(self:Motd Action Base Object Reference=Get Action()) -> (ReturnValue:bool))
   > n5 = Call Close Widget!(execute:Exec, self:WBP MOTD Full Tile Item Base Object Reference) -> (then:Exec) {is=title K2Node_CallDelegate}
   CallToAction!(
    execute=[@n5.then, @n4.else]
    self:Motd Action Base Object Reference=Get Action()
    PlayerController:Fort Player Controller Object Reference=@n3
   ) -> (then)
  ]}
 EventGraph {nodes [
   n1 = Get Buttons!() -> (OutputDelegate, then, Buttons Array:Array of Motd Button Structures) {is=title K2Node_CustomEvent}
   > Do Once()
   > n2 = Set Buttons Array(Buttons Array:Array of Motd Button Structures=@n1.Buttons Array) -> (Output_Get:Array of Motd Button Structures)
   > n3 = For Each Loop(Array:Array of Motd Button Structures=@n2) -> (Array Element:Motd Button Structure)
   n4 = Break Motd Button!(MotdButton:"Motd Button Structure (by ref)"=@n3) -> (Text:Text, Style:EMotdButtonStyle Enum, Action:Motd Action Base Object Reference) {is=title K2Node_BreakStruct}
   > Set Text!(execute, Text=@n4) -> (then, Output_Get)
   > n5 = Switch on EMotdButtonStyle!(execute, Selection:EMotdButtonStyle Enum=@n4.Style) -> (Main:Exec, Secondary:Exec, None:Exec) {is=title K2Node_SwitchEnum}
   > K2Node_Composite_0 = Loud Button!(execute) -> (then:Exec, ReturnValue:WBP UIKit Button Loud Object Reference) {is=title K2Node_Composite}
   > n6 = Set NewLoudButton(NewLoudButton:WBP UIKit Button Loud Object Reference=@K2Node_Composite_0) -> (Output_Get:WBP UIKit Button Loud Object Reference)
   > Set!(execute, Padding_Left:float=0, Padding_Top:float=16, Padding_Right:float=0, Padding_Bottom:float=0, self:User Widget Object Reference=@n6) -> (then, Output_Get:Margin Structure)
   > AddChild(self=Get VerticalBox_Buttons(), Content=Get NewLoudButton() -> (NewLoudButton:WBP UIKit Button Loud Object Reference))
   > n7 = SetHideInputAction!(execute, self:Common Button Base Object Reference=Get NewLoudButton(), bInHideInputAction:bool=true) -> (then)
   n8 = Loud Button Call Action!() -> (OutputDelegate, then, Button:Common Button Base Object Reference) {is=title K2Node_CustomEvent}
   Bind Event to On Clicked(execute=@n7.then, self:Common Button Base Object Reference=Get NewLoudButton(), Delegate=@n8) {is=title K2Node_AddDelegate}
   > Set Text!(execute, Text=Get Text!() -> (Text), self:Fort CTAButton Object Reference=Get NewLoudButton()) -> (then, Output_Get)
   K2Node_Composite_1 = Regular Button!(execute=@n5.Secondary) -> (then:Exec, ReturnValue:WBP UIKit Button Regular Object Reference) {is=title K2Node_Composite}
   > n9 = Set NewRegularButton(NewRegularButton:WBP UIKit Button Regular Object Reference=@K2Node_Composite_1) -> (Output_Get:WBP UIKit Button Regular Object Reference)
   > Set(Padding_Left=0, Padding_Top=16, Padding_Right=0, Padding_Bottom=0, self=@n9)
   > Set ButtonType!(
    execute
    ButtonType:E_UI_CTAButtonType Enum=NewEnumerator1
    self:WBP UIKit Button Regular Object Reference=Get NewRegularButton() -> (NewRegularButton:WBP UIKit Button Regular Object Reference)
   ) -> (then, Output_Get:E_UI_CTAButtonType Enum=NewEnumerator0)
   > SetHideInputAction(self=Get NewRegularButton(), bInHideInputAction=true)
   > n10 = AddChild(self=Get VerticalBox_Buttons(), Content=Get NewRegularButton())
   n11 = Regular Button Call Action!() -> (OutputDelegate, then, Button:Common Button Base Object Reference) {is=title K2Node_CustomEvent}
   Bind Event to On Clicked(execute=@n10.then, self=Get NewRegularButton(), Delegate=@n11)
   > Set Text(Text=Get Text(), self=Get NewRegularButton())
   n12 = DelayedVideoStart() {is=title K2Node_CustomEvent}
   > StartVideo!(execute, self, SkipClosedControlBroadcast:bool=Get StartVideoDelayed_Scrolled() -> (StartVideoDelayed_Scrolled:bool)) -> (then)
   > Set StartVideoDelayed_Scrolled(StartVideoDelayed_Scrolled:bool=false) -> (Output_Get:bool)
   n13 = StartVideo!() -> (OutputDelegate, then, SkipClosedControlBroadcast:bool=~) {is=title K2Node_CustomEvent}
   > n14 = Get!(execute:Exec) -> (then:Exec, else:Exec, MediaPlayer:Fort Dynamic Motd Media Player Object Reference)
   > n15 = Branch(Condition=Get VideoPlaying() -> (VideoPlaying:bool))
   SetPriority!(
    execute=@n15.else
    self:Fort Dynamic Motd Media Player Object Reference=@n14
    bInPriority:bool=Get VideoHasPriority() -> (VideoHasPriority:bool)
    bInSkipClosedControlBroadcast:bool=true
   ) -> (then)
   > Bind Event to Loaded(self:Fort Dynamic Motd Media Player Object Reference=@n14, Delegate=Create Event()) {is=title K2Node_AddDelegate}
   > Bind Event to Playing(self:Fort Dynamic Motd Media Player Object Reference=@n14, Delegate=Create Event()) {is=title K2Node_AddDelegate}
   > Bind Event to Ended(self:Fort Dynamic Motd Media Player Object Reference=@n14, Delegate=Create Event()) {is=title K2Node_AddDelegate}
   > Bind Event to Terminal Error Event(self:Fort Dynamic Motd Media Player Object Reference=@n14, Delegate=Create Event()) {is=title K2Node_AddDelegate}
   n16 = Get Motd() -> (Motd:Motd View Model Object Reference)
   > Start!(
    execute
    self:Fort Dynamic Motd Media Player Object Reference=@n14
    MotdVideo:"Motd Video Structure (by ref)"=Break Motd Background!(MotdBackground:"Motd Background Structure (by ref)"=Get FullScreenBackground!(self:Motd View Model Object Reference=@n16) -> (FullScreenBackground:Motd Background Structure)) -> (Video:Motd Video Structure) {is=title K2Node_BreakStruct}
    InMediaTexture:Media Texture Object Reference=/Motd/Widgets/Textures/MT_DynamicMOTD
    InMaxYResolution:int=-1
    bDisableCDNFailover:bool=false
    bManualLoop:bool=true
    InLoopOverride:int=-1
   ) -> (then)
   > Set VideoPlaying(VideoPlaying:bool=true) -> (Output_Get:bool)
   n17 = HandleVideoAction!() -> (OutputDelegate, then, Action:Motd Action Base Object Reference) {is=title K2Node_CustomEvent}
   > n18 = Cast To MotdVideoAction!(execute, Object=@n17.Action) -> (then, CastFailed, AsMotd Video Action:Motd Video Action Object Reference) {is=title K2Node_DynamicCast}
   n19 = VideoStartedEvent() {is=title K2Node_CustomEvent}
   n20 = Bind Event to Video Started(execute=@n18.then, self:Motd Video Action Object Reference=@n18, Delegate=@n19) {is=title K2Node_AddDelegate}
   n21 = VideoActionEndedEvent() {is=title K2Node_CustomEvent}
   Bind Event to Video Ended(execute=@n20.then, self:Motd Video Action Object Reference=@n18, Delegate=@n21) {is=title K2Node_AddDelegate}
   n22 = Get CMS Image!() -> (OutputDelegate, then, URL:String) {is=title K2Node_CustomEvent}
   > n23 = Construct Object from Class(Class=/Script/EpicCMSUIFramework.EpicCMSImage) -> (ReturnValue:Epic CMSImage Object Reference)
   n24 = OnImageLoadingComplete_Event() {is=title K2Node_CustomEvent}
   Assign On Image Loading Complete!(execute:Exec=@n23.then, self:Epic CMSImage Object Reference=@n23, Delegate:"Delegate (by ref)"=@n24) -> (then:Exec) {is=title K2Node_AssignDelegate}
   > SetMediaURL!(execute, self:Epic CMSImage Object Reference=@n23, MediaUrl:String=@n22.URL) -> (then)
   n25 = StopVideo!() -> (OutputDelegate, then, SkipClosedControlBroadcast:bool=~) {is=title K2Node_CustomEvent}
   > Clear and Invalidate Timer by Handle(Handle=Get StartVideoDelayed_Handle() -> (StartVideoDelayed_Handle:Timer Handle Structure))
   > n26 = Get()
   > Branch(Condition=Get VideoPlaying())
   > SetPriority(self=@n26, bInPriority=Get VideoHasPriority(), bInSkipClosedControlBroadcast=@n25.SkipClosedControlBroadcast)
   n27 = Break Motd Background!(MotdBackground=Get FullScreenBackground(self=Get Motd())) -> (Images:Array of Motd Image Structures)
   > Get CMS Image!(
    execute
    self
    URL:String=Break Motd Image!(MotdImage:"Motd Image Structure (by ref)"="Get (a copy)"(Array:Array of Motd Image Structures=@n27) -> (Output:Motd Image Structure)) -> (Url:String) {is=title K2Node_BreakStruct}
   ) -> (then) {is=title K2Node_CallFunction}
   > Stop(self:Fort Dynamic Motd Media Player Object Reference=@n26)
   > Set VideoPlaying(VideoPlaying=false)
   n28 = StartVideoDelayed!() -> (OutputDelegate, then, Delay:bool=~) {is=title K2Node_CustomEvent}
   > n29 = Branch(Condition=@n28.Delay)
   > Set StartVideoDelayed_Scrolled(StartVideoDelayed_Scrolled=@n28.Delay)
   > n30 = Set Timer by Event(Delegate=@n12, Time=0.5, bLooping=false)
   > Set StartVideoDelayed_Handle(StartVideoDelayed_Handle:Timer Handle Structure=@n30) -> (Output_Get:Timer Handle Structure)
   StartVideo(execute=@n29.else, SkipClosedControlBroadcast=@n28.Delay) {is=title K2Node_CallFunction}
   n31 = Event PreConstruct()
   > Parent: PreConstruct(IsDesignTime=@n31.IsDesignTime)
   > Branch(Condition=@n31.IsDesignTime)
   > Set Vertical Layout!(execute, self:WBP MOTD Full Tile Item Base Object Reference, IsVertical:bool=Get Debug_IsVertical() -> (Debug_IsVertical:bool)) -> (then) {is=title K2Node_CallFunction}
   n32 = Event OnVideoBackgroundsAllowedChanged!() -> (OutputDelegate, then, bAllowVideoBackgrounds:bool) {is=title K2Node_Event}
   n33 = Get MotdViewModel() -> (MotdViewModel:Motd View Model Object Reference)
   > Branch(Condition=AND Boolean(A=@n32.bAllowVideoBackgrounds, B=HasFullScreenVideoBackground!(self:Motd View Model Object Reference=@n33) -> (ReturnValue:bool)))
   > StartVideo(SkipClosedControlBroadcast=true)
   Event On Deactivated()
   > StopVideo(SkipClosedControlBroadcast=false) {is=title K2Node_CallFunction}
   PlayAnimation(execute=@n36.then, InAnimation=Get Loaded() -> (Loaded:Widget Animation Object Reference))
   PerformCallToAction!(execute=@n8.then, self, Button Context:Common Button Base Object Reference=@n8.Button) -> (then)
   PerformCallToAction(execute=@n11.then, Button Context=@n11.Button)
   SetActive!(execute=@n37.then, self:WBP MOTD Full Tile Item Base Object Reference, Active:bool=true, Scroll:bool=false) -> (then)
   Unbind all Events from Video Started!(execute:Exec=@n39.then, self:Motd Video Action Object Reference=@n18) -> (then:Exec) {is=title K2Node_ClearDelegate}
   Reroute Node(InputPin:bool=@n13.SkipClosedControlBroadcast) -> (OutputPin:bool)
   n34 = GetBrushResourceAsTexture2D!(Brush:"Slate Brush Structure (by ref)"=Get Brush!(self:Image Object Reference=@n23) -> (Brush:Slate Brush Structure)) -> (ReturnValue:Texture 2D Object Reference)
   n35 = GetDynamicMaterial(execute=@n40.then, self=Get Image() -> (Image:Image Object Reference))
   > n36 = SetTextureParameterValue(self=@n35, ParameterName=Image_A, Value=Get ImageTexture() -> (ImageTexture:Texture 2D Object Reference))
   n37 = Unbind Event from Video Ended(execute=@n38.then, self:Motd Video Action Object Reference=@n18, Delegate=@n21) {is=title K2Node_RemoveDelegate}
   n38 = Set DoingAction(execute=@n21.then, DoingAction:bool=false) -> (Output_Get:bool)
   n39 = Set DoingAction(execute=@n19.then, DoingAction=true)
   n40 = Set ImageTexture(execute=@n24.then, ImageTexture:Texture 2D Object Reference=@n34) -> (Output_Get:Texture 2D Object Reference)
  ]}
 Loud Button {
  path=EventGraph.K2Node_Composite_0.Loud Button
  nodes [
   Inputs()
   > n1 = Construct Object from Class!(
    execute
    Class=/FortUILibrary/UIKit/Buttons/WBP_UIKit_Button_Loud.WBP_UIKit_Button_Loud_C
    self
    Style:Common Button Style Class Reference=/FortUILibrary/UIKit/Buttons/Blocks/ButtonStyle_UIKit_Empty.ButtonStyle_UIKit_Empty_C
    bApplyAlphaOnDisable:bool=False
    bLocked:bool=False
    bSelectable:bool=False
    bShouldSelectUponReceivingFocus:bool=False
    bInteractableWhenSelected:bool=False
    bToggleable:bool=False
    bTriggerClickedAfterSelection:bool=False
    bDisplayInputActionWhenNotInteractable:bool=false
    bHideInputActionWithKeyboard:bool=true
    bShouldUseFallbackDefaultInputAction:bool=True
    bRequiresHold:bool=False
    HoldData:Common UIHold Data Class Reference=&CommonUI_HoldData_Base_C
    ClickMethod:EButtonClickMethod Enum=DownAndUp
    InputPriority:int=0
    TriggeringInputAction:Data Table Row Handle Structure
   ) -> (then, ReturnValue:WBP UIKit Button Loud Object Reference)
   > Outputs!(then:Exec, ReturnValue:WBP UIKit Button Loud Object Reference=@n1)
  ]
 }
 Regular Button {
  path=EventGraph.K2Node_Composite_1.Regular Button
  nodes [
   Inputs()
   > n1 = Construct Object from Class(
    Class=/FortUILibrary/UIKit/Buttons/WBP_UIKit_Button_Regular.WBP_UIKit_Button_Regular_C
    Style=&ButtonStyle_UIKit_Empty_C
    bApplyAlphaOnDisable=False
    bLocked=False
    bSelectable=False
    bShouldSelectUponReceivingFocus=False
    bInteractableWhenSelected=False
    bToggleable=False
    bTriggerClickedAfterSelection=False
    bDisplayInputActionWhenNotInteractable=false
    bHideInputActionWithKeyboard=true
    bShouldUseFallbackDefaultInputAction=True
    bRequiresHold=False
    HoldData=&CommonUI_HoldData_Base_C
    ClickMethod=DownAndUp
    InputPriority=0
   ) -> (ReturnValue:WBP UIKit Button Regular Object Reference)
   > Outputs(ReturnValue:WBP UIKit Button Regular Object Reference=@n1)
  ]
 }
 Get Background Image {nodes [
   n1 = Get Background Image!() -> (then, Motd Background:Motd Background Structure) {is=title K2Node_FunctionEntry}
   n2 = Break Motd Background(MotdBackground=@n1)
   > Get CMS Image(URL=Break Motd Image!(MotdImage="Get (a copy)"(Array=@n2)) -> (Url, Width:int, Height:int))
  ]}
 PerformCallToAction {nodes [
   n1 = PerformCallToAction!() -> (then, Button Context:Common Button Base Object Reference) {is=title K2Node_FunctionEntry}
   n2 = GetChildIndex!(self:Panel Widget Object Reference=Get VerticalBox_Buttons(), Content:Widget Object Reference=@n1) -> (ReturnValue:int=0)
   n3 = Get Buttons Array() -> (Buttons Array:Array of Motd Button Structures)
   > Branch(Condition=Is Valid Index!(TargetArray:Array of Motd Button Structures=@n3, IndexToTest:int=@n2) -> (ReturnValue:bool))
   n4 = "Get (a copy)"!(Array:Array of Motd Button Structures=@n3, Dimension 1=@n2) -> (Output_Text:Text, Output_Style:EMotdButtonStyle Enum, Output_Action:Motd Action Base Object Reference)
   > OnButtonPressed!(
    execute
    self:Fort Dynamic Motd Fullscreen Tile Object Reference
    Action:Motd Action Base Object Reference=@n4.Output_Action
    Style:EMotdButtonStyle Enum=@n4.Output_Style
   ) -> (then)
   > Button Pressed!(execute, self, Action:Motd Action Base Object Reference=@n4.Output_Action, Button Context:Common Button Base Object Reference=@n1) -> (then) {is=title K2Node_CallFunction}
  ]}
 Set Vertical Layout {nodes [
   n1 = Set Vertical Layout!() -> (then, IsVertical:bool) {is=title K2Node_FunctionEntry}
   > Parent: Set Vertical Layout!(execute, IsVertical:bool=@n1) -> (then) {is=title K2Node_CallParentFunction}
   > SetGridSlotPadding!(
    execute
    Widget:Widget Object Reference=Get GP_Content() -> (GP_Content:Grid Panel Object Reference)
    Padding:Margin Structure=Select!(
     Option 0:Margin Structure=Get GP_Content_Padding() -> (GP_Content_Padding:Margin Structure)
     Option 1_Left:float=48
     Option 1_Top:float=80
     Option 1_Right:float=48
     Option 1_Bottom:float=48
     Index=Get IsVertical() -> (IsVertical:bool)
    ) -> (ReturnValue:Margin Structure)
   ) -> (then)
   > SetVisibility(
    self=Get VB_Horizontal_Content() -> (VB_Horizontal_Content:Vertical Box Object Reference)
    InVisibility=Select!(Option 0:ESlateVisibility Enum=SelfHitTestInvisible, Option 1:ESlateVisibility Enum=Collapsed, Index=Get IsVertical()) -> (ReturnValue:ESlateVisibility Enum)
   )
   > SetVisibility(
    self=Get SB_ScrollingText() -> (SB_ScrollingText:Size Box Object Reference)
    InVisibility=Select(Option 0=Collapsed, Option 1=SelfHitTestInvisible, Index=Get IsVertical())
   )
   > SetWidthOverride(
    self=Get SizeBox_Buttons() -> (SizeBox_Buttons:Size Box Object Reference)
    InWidthOverride=Select(Option 0:float=420, Option 1:float=536, Index=Get IsVertical()) -> (ReturnValue:float)
   )
   > n2 = GetDynamicMaterial(self=Get Image())
   > SetScalarParameterValue(self=@n2, ParameterName=IsVertical, Value=SelectFloat(A=1, B=0, bPickA=Get IsVertical()))
   > SetScrollWidget!(execute, ToScroll:Widget Object Reference=Get ScrollBox() -> (ScrollBox:Scroll Box Object Reference)) -> (then)
  ]}
 SetActive {nodes [
   n1 = SetActive!() -> (then, Active:bool, Scroll:bool) {is=title K2Node_FunctionEntry}
   > Parent: SetActive!(execute, Active:bool=@n1, Scroll:bool=@n1.Scroll) -> (then) {is=title K2Node_CallParentFunction}
   > Set Button Index(Button Index:int=0) -> (Output_Get:int=0)
   > SpawnSound2D!(
    execute
    Sound:Sound Base Object Reference=Get IntroWhooshSound() -> (IntroWhooshSound:Sound Base Object Reference)
    VolumeMultiplier:float=1
    PitchMultiplier:float=1
    StartTime:float=0
    ConcurrencySettings:Sound Concurrency Object Reference
    bPersistAcrossLevelTransition:bool=false
    bAutoDestroy:bool=true
   ) -> (then, ReturnValue:Audio Component Object Reference)
   > n2 = GetDynamicMaterial(self=Get Image())
   > SetScalarParameterValue(self=@n2, ParameterName=GradientEnd, Value=Select(Option 0=2, Option 1=0.6, Index=Get Active() -> (Active:bool)))
   > SetVisibility(
    self=Get GP_Content()
    InVisibility=Select(Option 0:ESlateVisibility Enum=Collapsed, Option 1:ESlateVisibility Enum=SelfHitTestInvisible, Index=Get Active()) -> (ReturnValue:ESlateVisibility Enum)
   )
   > n3 = Branch(Condition=Get Active())
   > Branch(Condition=HasFullScreenVideoBackground(self=Get MotdViewModel()))
   > StartVideoDelayed!(execute, self, Delay:bool=Get Scroll() -> (Scroll:bool)) -> (then) {is=title K2Node_CallFunction}
   n4 = Branch(execute=@n3.else, Condition=HasFullScreenVideoBackground(self=Get MotdViewModel()))
   > n5 = StopVideo(SkipClosedControlBroadcast=Get Scroll())
   Set Vertical Layout(execute=[@n5.then, @n4.else], IsVertical=false) {is=title K2Node_CallFunction}
  ]}
 VideoEndedEvent {nodes [
   VideoEndedEvent() {is=title K2Node_FunctionEntry}
   > n1 = Get()
   > Unbind Event from Ended(self:Fort Dynamic Motd Media Player Object Reference=@n1, Delegate=Create Event()) {is=title K2Node_RemoveDelegate}
   > StopVideo(SkipClosedControlBroadcast=true)
  ]}
 VideoLoadedEvent {nodes [
   VideoLoadedEvent() {is=title K2Node_FunctionEntry}
   > n1 = Get()
   > Unbind all Events from Loaded!(execute:Exec, self:Fort Dynamic Motd Media Player Object Reference=@n1) -> (then:Exec) {is=title K2Node_ClearDelegate}
   > Play(self:Fort Dynamic Motd Media Player Object Reference=@n1)
  ]}
 VideoPlayingEvent {nodes [
   VideoPlayingEvent() {is=title K2Node_FunctionEntry}
   > n1 = Get()
   > n2 = Unbind all Events from Playing!(execute:Exec, self:Fort Dynamic Motd Media Player Object Reference=@n1) -> (then:Exec) {is=title K2Node_ClearDelegate}
   n3 = GetDynamicMaterial(execute=[@n2.then, @n1.else], self=Get Image())
   > SetTextureParameterValue(self=@n3, ParameterName=Image_A, Value=&MT_DynamicMOTD)
   > PlayAnimation(InAnimation=Get Loaded())
  ]}
 VideoTerminalErrorEvent {nodes [
   VideoTerminalErrorEvent!() -> (then, Reason:EBaseMediaTerminalErrorReason Enum=Unknown) {is=title K2Node_FunctionEntry}
   > n1 = Get()
   > n2 = Unbind all Events from Terminal Error Event!(execute:Exec, self:Fort Dynamic Motd Media Player Object Reference=@n1) -> (then:Exec) {is=title K2Node_ClearDelegate}
   Set VideoPlaying(execute=[@n2.then, @n1.else], VideoPlaying=false)
  ]}
]
variables [
 As Motd View Model {type=MotdViewModel, default=null}
 Button Index {type=integer}
 Buttons Array {type="array<MotdButton>", default []}
 Debug_IsVertical {type=boolean}
 DoingAction {type=boolean}
 GP_Content_Padding {type=Margin, default=<ltrb 80 80 80 80>}
 ImageTexture {type=Texture2D}
 Index {type=integer, default=1}
 Loading {type=boolean, default=true}
 Margin_CompactUI {type=Margin, default=<ltrb 48 48 48 48>}
 Margin_Default {type=Margin, default=<ltrb 80 80 80 80>}
 Material {type=MaterialInstanceDynamic}
 Motd Buttons {type="array<MotdButton>", default []}
 NewLoudButton {type=WBP_UIKit_Button_Loud_C, default=null}
 NewRegularButton {type=WBP_UIKit_Button_Regular_C, default=null}
 StartVideoDelayed_Handle {type=TimerHandle}
 StartVideoDelayed_Scrolled {type=boolean}
 Teaser Images {type="array<Texture2D>", default []}
 Text {type=string, default=""}
 VideoHasPriority {type=boolean, default=true}
 VideoPlaying {type=boolean}
]
widgets {
 info {
  inheritedWidgetCount=0
  namedSlotCount=1
  parentClass=&/WBP_MOTD_Full_TileItem_Base.WBP_MOTD_Full_TileItem_Base_C
  rootWidgetClass=null
  widgetCount=16
 }
 widgets [
  Overlay_56 {widgetClassPath=&Overlay, widget=Overlay_56}
  GridPanel_0 {
   widgetClassPath=&GridPanel
   parent=Overlay_56
   widget=GridPanel_0
   properties {columnFill [1], rowFill [1]}
   slot {class=OverlaySlot, properties {horizontalAlignment=HAlign_Fill, verticalAlignment=VAlign_Fill, padding=<ltrb 20 0 20 0>}}
  }
  S_ItemSize {
   widgetClassPath=&Spacer
   bIsVariable=true
   parent=GridPanel_0
   widget=S_ItemSize
   properties {size=<1648 920>}
   slot {class=GridSlot, properties {layer=-20}}
  }
  Image {
   widgetClassPath=&Image
   bIsVariable=true
   parent=GridPanel_0
   widget=Image
   properties {
    bIsVolatile=true
    brush {
     bIsDynamicallyLoaded=false
     drawAs=Image
     imageType=FullColor
     mirroring=NoMirror
     resourceName=None
     resourceObject=/Motd/Widgets/Materials/MI_MOTD_Full_Tile
     tiling=NoTile
     imageSize=<32 32>
     margin=<ltrb 0.1 0.1 0.1 0.1>
     outlineSettings {
      bUseBrushTransparency=false
      roundingType=HalfHeightRadius
      width=0.2
      color {colorUseRule=UseColor_Specified, specifiedColor=<rgba 0 0 0 0>}
      cornerRadii=<0.2 0.2 0.2 0.2>
     }
     tintColor {colorUseRule=UseColor_Specified, specifiedColor=<rgba 1 1 1 1>}
     uVRegion {bIsValid=false, max=<0 0>, min=<0 0>}
    }
    colorAndOpacity=<rgba 1 1 1 0>
   }
   slot {class=GridSlot}
  }
  WBP_CaptureForPostBufferUpdate {
   widgetClassPath=&WBP_CaptureForPostBufferUpdate
   bIsVariable=true
   parent=GridPanel_0
   properties {bIsFocusable=true, navigation=&self:WidgetTree.WBP_CaptureForPostBufferUpdate.WidgetNavigation_0}
   slot {class=GridSlot, properties {layer=1}}
  }
  GP_Content {
   widgetClassPath=&GridPanel
   bIsVariable=true
   parent=GridPanel_0
   widget=GP_Content
   properties {columnFill [1,0], rowFill [1]}
   slot {class=GridSlot, properties {layer=10, padding=<ltrb 80 80 80 80>}}
  }
  SB_ScrollingText {
   widgetClassPath=&SizeBox
   bIsVariable=true
   parent=GP_Content
   widget=SB_ScrollingText
   properties {bOverride_WidthOverride=true, widthOverride=536}
   slot {class=GridSlot, properties {column=1, verticalAlignment=VAlign_Bottom, padding=<ltrb 0 0 0 32>}}
  }
  ScrollBox {
   widgetClassPath=&ScrollBox
   bIsVariable=true
   parent=SB_ScrollingText
   widget=ScrollBox
   properties {
    allowOverscroll=false
    bAllowRightClickDragScrolling=false
    bAnimateWheelScrolling=true
    scrollWhenFocusChanges=AnimatedScroll
    scrollbarPadding=<ltrb 12 0 0 0>
    scrollbarThickness=<10 10>
    widgetBarStyle {
     thickness=10
     draggedThumbImage {
      bIsDynamicallyLoaded=false
      drawAs=RoundedBox
      imageType=FullColor
      mirroring=NoMirror
      resourceName=../../../Engine/Content/Slate/Common/Scrollbar_Thumb.png
      resourceObject=null
      tiling=NoTile
      imageSize=<12 12>
      margin=<ltrb 0.25 0.25 0.25 0.25>
      outlineSettings {
       bUseBrushTransparency=false
       roundingType=FixedRadius
       width=0
       color {colorUseRule=UseColor_Specified, specifiedColor=<rgba 0 0 0 0>}
       cornerRadii=<6 6 6 6>
      }
      tintColor {colorUseRule=UseColor_Specified, specifiedColor=<rgba 1 1 1 1>}
      uVRegion {bIsValid=false, max=<0 0>, min=<0 0>}
     }
     horizontalBackgroundImage {
      bIsDynamicallyLoaded=false
      drawAs=NoDrawType
      imageType=NoImage
      mirroring=NoMirror
      resourceName=None
      resourceObject=null
      tiling=NoTile
      imageSize=<0 0>
      margin=<ltrb 0 0 0 0>
      outlineSettings {
       bUseBrushTransparency=false
       roundingType=HalfHeightRadius
       width=0
       color {colorUseRule=UseColor_Specified, specifiedColor=<rgba 0 0 0 0>}
       cornerRadii=<0 0 0 0>
      }
      tintColor {colorUseRule=UseColor_Specified, specifiedColor=<rgba 1 1 1 1>}
      uVRegion {bIsValid=false, max=<0 0>, min=<0 0>}
     }
     horizontalBottomSlotImage {
      bIsDynamicallyLoaded=false
      drawAs=NoDrawType
      imageType=FullColor
      mirroring=NoMirror
      resourceName=../../../Engine/Content/Slate/Common/Scrollbar_Background_Horizontal.png
      resourceObject=null
      tiling=NoTile
      imageSize=<8 8>
      margin=<ltrb 0 0 0 0>
      outlineSettings {
       bUseBrushTransparency=false
       roundingType=HalfHeightRadius
       width=0
       color {colorUseRule=UseColor_Specified, specifiedColor=<rgba 0 0 0 0>}
       cornerRadii=<0 0 0 0>
      }
      tintColor {colorUseRule=UseColor_Specified, specifiedColor=<rgba 1 1 1 1>}
      uVRegion {bIsValid=false, max=<0 0>, min=<0 0>}
     }
     horizontalTopSlotImage {
      bIsDynamicallyLoaded=false
      drawAs=NoDrawType
      imageType=FullColor
      mirroring=NoMirror
      resourceName=../../../Engine/Content/Slate/Common/Scrollbar_Background_Horizontal.png
      resourceObject=null
      tiling=NoTile
      imageSize=<8 8>
      margin=<ltrb 0 0 0 0>
      outlineSettings {
       bUseBrushTransparency=false
       roundingType=HalfHeightRadius
       width=0
       color {colorUseRule=UseColor_Specified, specifiedColor=<rgba 0 0 0 0>}
       cornerRadii=<0 0 0 0>
      }
      tintColor {colorUseRule=UseColor_Specified, specifiedColor=<rgba 1 1 1 1>}
      uVRegion {bIsValid=false, max=<0 0>, min=<0 0>}
     }
     hoveredThumbImage {
      bIsDynamicallyLoaded=false
      drawAs=RoundedBox
      imageType=FullColor
      mirroring=NoMirror
      resourceName=../../../Engine/Content/Slate/Common/Scrollbar_Thumb.png
      resourceObject=null
      tiling=NoTile
      imageSize=<12 12>
      margin=<ltrb 0.25 0.25 0.25 0.25>
      outlineSettings {
       bUseBrushTransparency=false
       roundingType=FixedRadius
       width=0
       color {colorUseRule=UseColor_Specified, specifiedColor=<rgba 0 0 0 0>}
       cornerRadii=<6 6 6 6>
      }
      tintColor {colorUseRule=UseColor_Specified, specifiedColor=<rgba 1 1 1 1>}
      uVRegion {bIsValid=false, max=<0 0>, min=<0 0>}
     }
     normalThumbImage {
      bIsDynamicallyLoaded=false
      drawAs=RoundedBox
      imageType=FullColor
      mirroring=NoMirror
      resourceName=../../../Engine/Content/Slate/Common/Scrollbar_Thumb.png
      resourceObject=null
      tiling=NoTile
      imageSize=<12 12>
      margin=<ltrb 0.25 0.25 0.25 0.25>
      outlineSettings {
       bUseBrushTransparency=false
       roundingType=FixedRadius
       width=0
       color {colorUseRule=UseColor_Specified, specifiedColor=<rgba 0 0 0 0>}
       cornerRadii=<6 6 6 6>
      }
      tintColor {colorUseRule=UseColor_Specified, specifiedColor=<rgba 1 1 1 0.75>}
      uVRegion {bIsValid=false, max=<0 0>, min=<0 0>}
     }
     verticalBackgroundImage {
      bIsDynamicallyLoaded=false
      drawAs=Box
      imageType=NoImage
      mirroring=NoMirror
      resourceName=None
      resourceObject=/FortUILibrary/UIKit/ScrollBar/MaterialInstances/MI_UIKit_ScrollbarBackground
      tiling=NoTile
      imageSize=<12 12>
      margin=<ltrb 0.5 0.5 0.5 0.5>
      outlineSettings {
       bUseBrushTransparency=false
       roundingType=HalfHeightRadius
       width=0
       color {colorUseRule=UseColor_Specified, specifiedColor=<rgba 0 0 0 0>}
       cornerRadii=<0 0 0 0>
      }
      tintColor {colorUseRule=UseColor_Specified, specifiedColor=<rgba 1 1 1 1>}
      uVRegion {bIsValid=false, max=<0 0>, min=<0 0>}
     }
     verticalBottomSlotImage {
      bIsDynamicallyLoaded=false
      drawAs=NoDrawType
      imageType=FullColor
      mirroring=NoMirror
      resourceName=../../../Engine/Content/Slate/Common/Scrollbar_Background_Vertical.png
      resourceObject=null
      tiling=NoTile
      imageSize=<8 8>
      margin=<ltrb 0 0 0 0>
      outlineSettings {
       bUseBrushTransparency=false
       roundingType=HalfHeightRadius
       width=0
       color {colorUseRule=UseColor_Specified, specifiedColor=<rgba 0 0 0 0>}
       cornerRadii=<0 0 0 0>
      }
      tintColor {colorUseRule=UseColor_Specified, specifiedColor=<rgba 1 1 1 1>}
      uVRegion {bIsValid=false, max=<0 0>, min=<0 0>}
     }
     verticalTopSlotImage {
      bIsDynamicallyLoaded=false
      drawAs=NoDrawType
      imageType=FullColor
      mirroring=NoMirror
      resourceName=../../../Engine/Content/Slate/Common/Scrollbar_Background_Vertical.png
      resourceObject=null
      tiling=NoTile
      imageSize=<8 8>
      margin=<ltrb 0 0 0 0>
      outlineSettings {
       bUseBrushTransparency=false
       roundingType=HalfHeightRadius
       width=0
       color {colorUseRule=UseColor_Specified, specifiedColor=<rgba 0 0 0 0>}
       cornerRadii=<0 0 0 0>
      }
      tintColor {colorUseRule=UseColor_Specified, specifiedColor=<rgba 1 1 1 1>}
      uVRegion {bIsValid=false, max=<0 0>, min=<0 0>}
     }
    }
    widgetStyle {
     barThickness=12
     bottomShadowBrush {
      bIsDynamicallyLoaded=false
      drawAs=NoDrawType
      imageType=FullColor
      mirroring=NoMirror
      resourceName=../../../Engine/Content/Slate/Common/ScrollBoxShadowBottom.png
      resourceObject=null
      tiling=NoTile
      imageSize=<16 8>
      margin=<ltrb 0.5 0 0.5 1>
      outlineSettings {
       bUseBrushTransparency=false
       roundingType=HalfHeightRadius
       width=0
       color {colorUseRule=UseColor_Specified, specifiedColor=<rgba 0 0 0 0>}
       cornerRadii=<0 0 0 0>
      }
      tintColor {colorUseRule=UseColor_Specified, specifiedColor=<rgba 1 1 1 1>}
      uVRegion {bIsValid=false, max=<0 0>, min=<0 0>}
     }
     horizontalScrolledContentPadding=<ltrb 0 0 1 0>
     leftShadowBrush {
      bIsDynamicallyLoaded=false
      drawAs=NoDrawType
      imageType=FullColor
      mirroring=NoMirror
      resourceName=../../../Engine/Content/Slate/Common/ScrollBoxShadowLeft.png
      resourceObject=null
      tiling=NoTile
      imageSize=<8 16>
      margin=<ltrb 1 0.5 0 0.5>
      outlineSettings {
       bUseBrushTransparency=false
       roundingType=HalfHeightRadius
       width=0
       color {colorUseRule=UseColor_Specified, specifiedColor=<rgba 0 0 0 0>}
       cornerRadii=<0 0 0 0>
      }
      tintColor {colorUseRule=UseColor_Specified, specifiedColor=<rgba 1 1 1 1>}
      uVRegion {bIsValid=false, max=<0 0>, min=<0 0>}
     }
     rightShadowBrush {
      bIsDynamicallyLoaded=false
      drawAs=NoDrawType
      imageType=FullColor
      mirroring=NoMirror
      resourceName=../../../Engine/Content/Slate/Common/ScrollBoxShadowRight.png
      resourceObject=null
      tiling=NoTile
      imageSize=<8 16>
      margin=<ltrb 0 0.5 1 0.5>
      outlineSettings {
       bUseBrushTransparency=false
       roundingType=HalfHeightRadius
       width=0
       color {colorUseRule=UseColor_Specified, specifiedColor=<rgba 0 0 0 0>}
       cornerRadii=<0 0 0 0>
      }
      tintColor {colorUseRule=UseColor_Specified, specifiedColor=<rgba 1 1 1 1>}
      uVRegion {bIsValid=false, max=<0 0>, min=<0 0>}
     }
     topShadowBrush {
      bIsDynamicallyLoaded=false
      drawAs=NoDrawType
      imageType=FullColor
      mirroring=NoMirror
      resourceName=../../../Engine/Content/Slate/Common/ScrollBoxShadowTop.png
      resourceObject=null
      tiling=NoTile
      imageSize=<16 8>
      margin=<ltrb 0.5 1 0.5 0>
      outlineSettings {
       bUseBrushTransparency=false
       roundingType=HalfHeightRadius
       width=0
       color {colorUseRule=UseColor_Specified, specifiedColor=<rgba 0 0 0 0>}
       cornerRadii=<0 0 0 0>
      }
      tintColor {colorUseRule=UseColor_Specified, specifiedColor=<rgba 1 1 1 1>}
      uVRegion {bIsValid=false, max=<0 0>, min=<0 0>}
     }
     verticalScrolledContentPadding=<ltrb 0 0 0 1>
    }
   }
   slot {class=SizeBoxSlot}
  }
  VB_Vertical_Content {
   widgetClassPath=&VerticalBox
   bIsVariable=true
   parent=ScrollBox
   widget=VB_Vertical_Content
   properties {flowDirectionPreference=Culture}
   slot {class=ScrollBoxSlot}
  }
  Title_Vertical {
   widgetClassPath=&CommonRichTextBlock
   bIsVariable=true
   parent=VB_Vertical_Content
   widget=Title_Vertical
   properties {
    autoWrapText=true
    bIsScrollingEnabled=false
    flowDirectionPreference=Culture
    text=VERTICAL TITLE
    textStyleSet=/FortniteCoreUI/Frontend/MOTD/MOTDWidgetTitleStyles
    defaultTextStyleOverride {
     overflowPolicy=Clip
     transformPolicy=None
     colorAndOpacity {colorUseRule=UseColor_Specified, specifiedColor=<rgba 1 1 1 1>}
     font {
      bForceMonospaced=false
      bMaterialIsStencil=false
      fontMaterial=null
      fontObject=&Burbank
      letterSpacing=0
      monospacedWidth=1
      size=64
      skewAmount=0
      typefaceFontName=Burbank-Black
      outlineSettings {
       bApplyOutlineToDropShadows=false
       bMiteredCorners=false
       bSeparateFillAlpha=false
       outlineMaterial=null
       outlineSize=0
       outlineColor=<rgba 0 0 0 1>
      }
     }
     highlightColor {colorUseRule=UseColor_Specified, specifiedColor=<rgba 0 0 0 1>}
     highlightShape {
      bIsDynamicallyLoaded=false
      drawAs=Image
      imageType=NoImage
      mirroring=NoMirror
      resourceName=None
      resourceObject=null
      tiling=NoTile
      imageSize=<32 32>
      margin=<ltrb 0 0 0 0>
      outlineSettings {
       bUseBrushTransparency=false
       roundingType=HalfHeightRadius
       width=0
       color {colorUseRule=UseColor_Specified, specifiedColor=<rgba 0 0 0 0>}
       cornerRadii=<0 0 0 0>
      }
      tintColor {colorUseRule=UseColor_Specified, specifiedColor=<rgba 1 1 1 1>}
      uVRegion {bIsValid=false, max=<0 0>, min=<0 0>}
     }
     selectedBackgroundColor {colorUseRule=UseColor_Foreground, specifiedColor=<rgba 1 0 1 1>}
     shadowColorAndOpacity=<rgba 0 0 0 1>
     shadowOffset=<0 0>
     strikeBrush {
      bIsDynamicallyLoaded=false
      drawAs=Image
      imageType=NoImage
      mirroring=NoMirror
      resourceName=None
      resourceObject=null
      tiling=NoTile
      imageSize=<32 32>
      margin=<ltrb 0 0 0 0>
      outlineSettings {
       bUseBrushTransparency=false
       roundingType=HalfHeightRadius
       width=0
       color {colorUseRule=UseColor_Specified, specifiedColor=<rgba 0 0 0 0>}
       cornerRadii=<0 0 0 0>
      }
      tintColor {colorUseRule=UseColor_Specified, specifiedColor=<rgba 1 1 1 1>}
      uVRegion {bIsValid=false, max=<0 0>, min=<0 0>}
     }
     underlineBrush {
      bIsDynamicallyLoaded=false
      drawAs=Image
      imageType=NoImage
      mirroring=NoMirror
      resourceName=None
      resourceObject=null
      tiling=NoTile
      imageSize=<32 32>
      margin=<ltrb 0 0 0 0>
      outlineSettings {
       bUseBrushTransparency=false
       roundingType=HalfHeightRadius
       width=0
       color {colorUseRule=UseColor_Specified, specifiedColor=<rgba 0 0 0 0>}
       cornerRadii=<0 0 0 0>
      }
      tintColor {colorUseRule=UseColor_Specified, specifiedColor=<rgba 1 1 1 1>}
      uVRegion {bIsValid=false, max=<0 0>, min=<0 0>}
     }
    }
    shapedTextOptions {bOverride_TextFlowDirection=true, bOverride_TextShapingMethod=false, textFlowDirection=Culture, textShapingMethod=Auto}
   }
   slot {class=VerticalBoxSlot, properties {verticalAlignment=VAlign_Top, padding=<ltrb 0 0 0 16>}}
  }
  Content_Vertical {
   widgetClassPath=&CommonRichTextBlock
   bIsVariable=true
   parent=VB_Vertical_Content
   widget=Content_Vertical
   properties {
    autoWrapText=true
    bIsScrollingEnabled=false
    flowDirectionPreference=Culture
    mobileTextBlockScale=1.14
    text="Lorem ipsum dolor sit amet, consectetur adipiscing elit. Donec vulputate mauris ac dui bibendum, sed tristique justo porta. Nullam sodales enim eget ante pulvinar mattis. Nullam facilisis viverra neque. Vivamus volutpat nisi et lorem semper, ac malesuada dolor tincidunt. Donec tristique dui eget blandit accumsan. Aenean finibus ultricies eros, vitae scelerisque ante vulputate tempus. Class aptent taciti sociosqu ad litora torquent per conubia nostra, per inceptos himenaeos. Integer eu turpis cursus, tincidunt nibh eget, imperdiet mauris. Sed odio mi, consequat et tincidunt in, condimentum sed nisl. Nunc elementum vehicula magna, sed volutpat augue finibus pharetra. In justo lacus, venenatis ut blandit eget, rutrum sed ipsum. Nam lacus nulla, commodo nec imperdiet vel, aliquam nec erat. Vestibulum placerat porta imperdiet. Suspendisse a elit augue.Lorem ipsum dolor sit amet, consectetur adipiscing elit. Donec vulputate mauris ac dui bibendum, sed tristique justo porta. Nullam sodales enim eget ante pulvinar mattis. Nullam facilisis viverra neque. Vivamus volutpat nisi et lorem semper, ac malesuada dolor tincidunt. Donec tristique dui eget blandit accumsan. Aenean finibus ultricies eros, vitae scelerisque ante vulputate tempus. Class aptent taciti sociosqu ad litora torquent per conubia nostra, per inceptos himenaeos. Integer eu turpis cursus, tincidunt nibh eget, imperdiet mauris. Sed odio mi, consequat et tincidunt in, condimentum sed nisl. Nunc elementum vehicula magna, sed volutpat augue finibus pharetra. In justo lacus, venenatis ut blandit eget, rutrum sed ipsum. Nam lacus nulla, commodo nec imperdiet vel, aliquam nec erat. Vestibulum placerat porta imperdiet. Suspendisse a elit augue."
    textStyleSet=/FortniteCoreUI/Frontend/MOTD/MOTDWidgetBodyStyles
    defaultTextStyleOverride {
     overflowPolicy=Clip
     transformPolicy=None
     colorAndOpacity {colorUseRule=UseColor_Specified, specifiedColor=<rgba 0.6038 1 1 1>}
     font {
      bForceMonospaced=false
      bMaterialIsStencil=false
      fontMaterial=null
      fontObject=&BurbankSmall
      letterSpacing=0
      monospacedWidth=1
      size=24
      skewAmount=0
      typefaceFontName=Bold
      outlineSettings {
       bApplyOutlineToDropShadows=false
       bMiteredCorners=false
       bSeparateFillAlpha=false
       outlineMaterial=null
       outlineSize=0
       outlineColor=<rgba 0 0 0 1>
      }
     }
     highlightColor {colorUseRule=UseColor_Specified, specifiedColor=<rgba 0 0 0 1>}
     highlightShape {
      bIsDynamicallyLoaded=false
      drawAs=Image
      imageType=NoImage
      mirroring=NoMirror
      resourceName=None
      resourceObject=null
      tiling=NoTile
      imageSize=<32 32>
      margin=<ltrb 0 0 0 0>
      outlineSettings {
       bUseBrushTransparency=false
       roundingType=HalfHeightRadius
       width=0
       color {colorUseRule=UseColor_Specified, specifiedColor=<rgba 0 0 0 0>}
       cornerRadii=<0 0 0 0>
      }
      tintColor {colorUseRule=UseColor_Specified, specifiedColor=<rgba 1 1 1 1>}
      uVRegion {bIsValid=false, max=<0 0>, min=<0 0>}
     }
     selectedBackgroundColor {colorUseRule=UseColor_Foreground, specifiedColor=<rgba 1 0 1 1>}
     shadowColorAndOpacity=<rgba 0 0 0 1>
     shadowOffset=<0 0>
     strikeBrush {
      bIsDynamicallyLoaded=false
      drawAs=Image
      imageType=NoImage
      mirroring=NoMirror
      resourceName=None
      resourceObject=null
      tiling=NoTile
      imageSize=<32 32>
      margin=<ltrb 0 0 0 0>
      outlineSettings {
       bUseBrushTransparency=false
       roundingType=HalfHeightRadius
       width=0
       color {colorUseRule=UseColor_Specified, specifiedColor=<rgba 0 0 0 0>}
       cornerRadii=<0 0 0 0>
      }
      tintColor {colorUseRule=UseColor_Specified, specifiedColor=<rgba 1 1 1 1>}
      uVRegion {bIsValid=false, max=<0 0>, min=<0 0>}
     }
     underlineBrush {
      bIsDynamicallyLoaded=false
      drawAs=Image
      imageType=NoImage
      mirroring=NoMirror
      resourceName=None
      resourceObject=null
      tiling=NoTile
      imageSize=<32 32>
      margin=<ltrb 0 0 0 0>
      outlineSettings {
       bUseBrushTransparency=false
       roundingType=HalfHeightRadius
       width=0
       color {colorUseRule=UseColor_Specified, specifiedColor=<rgba 0 0 0 0>}
       cornerRadii=<0 0 0 0>
      }
      tintColor {colorUseRule=UseColor_Specified, specifiedColor=<rgba 1 1 1 1>}
      uVRegion {bIsValid=false, max=<0 0>, min=<0 0>}
     }
    }
    shapedTextOptions {bOverride_TextFlowDirection=true, bOverride_TextShapingMethod=false, textFlowDirection=Culture, textShapingMethod=Auto}
   }
   slot {class=VerticalBoxSlot, properties {verticalAlignment=VAlign_Top}}
  }
  VB_Horizontal_Content {
   widgetClassPath=&VerticalBox
   bIsVariable=true
   parent=GP_Content
   widget=VB_Horizontal_Content
   properties {flowDirectionPreference=Culture}
   slot {class=GridSlot, properties {rowSpan=2, verticalAlignment=VAlign_Bottom, padding=<ltrb 0 0 24 0>}}
  }
  Title_Horizontal {
   widgetClassPath=&CommonRichTextBlock
   bIsVariable=true
   parent=VB_Horizontal_Content
   widget=Title_Horizontal
   properties {
    autoWrapText=true
    bIsScrollingEnabled=false
    flowDirectionPreference=Culture
    text=TITLE
    textStyleSet=&MOTDWidgetTitleStyles
    wrapTextAt=1040
    defaultTextStyleOverride {
     overflowPolicy=Clip
     transformPolicy=None
     colorAndOpacity {colorUseRule=UseColor_Specified, specifiedColor=<rgba 1 1 1 1>}
     font {
      bForceMonospaced=false
      bMaterialIsStencil=false
      fontMaterial=null
      fontObject=&Burbank
      letterSpacing=0
      monospacedWidth=1
      size=64
      skewAmount=0
      typefaceFontName=Burbank-Black
      outlineSettings {
       bApplyOutlineToDropShadows=false
       bMiteredCorners=false
       bSeparateFillAlpha=false
       outlineMaterial=null
       outlineSize=0
       outlineColor=<rgba 0 0 0 1>
      }
     }
     highlightColor {colorUseRule=UseColor_Specified, specifiedColor=<rgba 0 0 0 1>}
     highlightShape {
      bIsDynamicallyLoaded=false
      drawAs=Image
      imageType=NoImage
      mirroring=NoMirror
      resourceName=None
      resourceObject=null
      tiling=NoTile
      imageSize=<32 32>
      margin=<ltrb 0 0 0 0>
      outlineSettings {
       bUseBrushTransparency=false
       roundingType=HalfHeightRadius
       width=0
       color {colorUseRule=UseColor_Specified, specifiedColor=<rgba 0 0 0 0>}
       cornerRadii=<0 0 0 0>
      }
      tintColor {colorUseRule=UseColor_Specified, specifiedColor=<rgba 1 1 1 1>}
      uVRegion {bIsValid=false, max=<0 0>, min=<0 0>}
     }
     selectedBackgroundColor {colorUseRule=UseColor_Foreground, specifiedColor=<rgba 1 0 1 1>}
     shadowColorAndOpacity=<rgba 0 0 0 1>
     shadowOffset=<0 0>
     strikeBrush {
      bIsDynamicallyLoaded=false
      drawAs=Image
      imageType=NoImage
      mirroring=NoMirror
      resourceName=None
      resourceObject=null
      tiling=NoTile
      imageSize=<32 32>
      margin=<ltrb 0 0 0 0>
      outlineSettings {
       bUseBrushTransparency=false
       roundingType=HalfHeightRadius
       width=0
       color {colorUseRule=UseColor_Specified, specifiedColor=<rgba 0 0 0 0>}
       cornerRadii=<0 0 0 0>
      }
      tintColor {colorUseRule=UseColor_Specified, specifiedColor=<rgba 1 1 1 1>}
      uVRegion {bIsValid=false, max=<0 0>, min=<0 0>}
     }
     underlineBrush {
      bIsDynamicallyLoaded=false
      drawAs=Image
      imageType=NoImage
      mirroring=NoMirror
      resourceName=None
      resourceObject=null
      tiling=NoTile
      imageSize=<32 32>
      margin=<ltrb 0 0 0 0>
      outlineSettings {
       bUseBrushTransparency=false
       roundingType=HalfHeightRadius
       width=0
       color {colorUseRule=UseColor_Specified, specifiedColor=<rgba 0 0 0 0>}
       cornerRadii=<0 0 0 0>
      }
      tintColor {colorUseRule=UseColor_Specified, specifiedColor=<rgba 1 1 1 1>}
      uVRegion {bIsValid=false, max=<0 0>, min=<0 0>}
     }
    }
    shapedTextOptions {bOverride_TextFlowDirection=true, bOverride_TextShapingMethod=false, textFlowDirection=Culture, textShapingMethod=Auto}
   }
   slot {class=VerticalBoxSlot, properties {verticalAlignment=VAlign_Bottom, padding=<ltrb 0 0 0 16>, size {sizeRule=Fill, value=1}}}
  }
  Content_Horizontal {
   widgetClassPath=&CommonRichTextBlock
   bIsVariable=true
   parent=VB_Horizontal_Content
   widget=Content_Horizontal
   properties {
    applyLineHeightToBottomLine=false
    autoWrapText=true
    bIsScrollingEnabled=false
    flowDirectionPreference=Culture
    mobileTextBlockScale=1.14
    text="{PH} This text <PurpleBold>can expand</> to this long <RedBold>for a description</> of <White>what the news</> is, brevity is best\r\nText has a horizontal limit but scales vertically"
    textStyleSet=&MOTDWidgetBodyStyles
    wrapTextAt=1040
    defaultTextStyleOverride {
     overflowPolicy=Clip
     transformPolicy=None
     colorAndOpacity {colorUseRule=UseColor_Specified, specifiedColor=<rgba 0.6038 1 1 1>}
     font {
      bForceMonospaced=false
      bMaterialIsStencil=false
      fontMaterial=null
      fontObject=&BurbankSmall
      letterSpacing=0
      monospacedWidth=1
      size=24
      skewAmount=0
      typefaceFontName=Bold
      outlineSettings {
       bApplyOutlineToDropShadows=false
       bMiteredCorners=false
       bSeparateFillAlpha=false
       outlineMaterial=null
       outlineSize=0
       outlineColor=<rgba 0 0 0 1>
      }
     }
     highlightColor {colorUseRule=UseColor_Specified, specifiedColor=<rgba 0 0 0 1>}
     highlightShape {
      bIsDynamicallyLoaded=false
      drawAs=Image
      imageType=NoImage
      mirroring=NoMirror
      resourceName=None
      resourceObject=null
      tiling=NoTile
      imageSize=<32 32>
      margin=<ltrb 0 0 0 0>
      outlineSettings {
       bUseBrushTransparency=false
       roundingType=HalfHeightRadius
       width=0
       color {colorUseRule=UseColor_Specified, specifiedColor=<rgba 0 0 0 0>}
       cornerRadii=<0 0 0 0>
      }
      tintColor {colorUseRule=UseColor_Specified, specifiedColor=<rgba 1 1 1 1>}
      uVRegion {bIsValid=false, max=<0 0>, min=<0 0>}
     }
     selectedBackgroundColor {colorUseRule=UseColor_Foreground, specifiedColor=<rgba 1 0 1 1>}
     shadowColorAndOpacity=<rgba 0 0 0 1>
     shadowOffset=<0 0>
     strikeBrush {
      bIsDynamicallyLoaded=false
      drawAs=Image
      imageType=NoImage
      mirroring=NoMirror
      resourceName=None
      resourceObject=null
      tiling=NoTile
      imageSize=<32 32>
      margin=<ltrb 0 0 0 0>
      outlineSettings {
       bUseBrushTransparency=false
       roundingType=HalfHeightRadius
       width=0
       color {colorUseRule=UseColor_Specified, specifiedColor=<rgba 0 0 0 0>}
       cornerRadii=<0 0 0 0>
      }
      tintColor {colorUseRule=UseColor_Specified, specifiedColor=<rgba 1 1 1 1>}
      uVRegion {bIsValid=false, max=<0 0>, min=<0 0>}
     }
     underlineBrush {
      bIsDynamicallyLoaded=false
      drawAs=Image
      imageType=NoImage
      mirroring=NoMirror
      resourceName=None
      resourceObject=null
      tiling=NoTile
      imageSize=<32 32>
      margin=<ltrb 0 0 0 0>
      outlineSettings {
       bUseBrushTransparency=false
       roundingType=HalfHeightRadius
       width=0
       color {colorUseRule=UseColor_Specified, specifiedColor=<rgba 0 0 0 0>}
       cornerRadii=<0 0 0 0>
      }
      tintColor {colorUseRule=UseColor_Specified, specifiedColor=<rgba 1 1 1 1>}
      uVRegion {bIsValid=false, max=<0 0>, min=<0 0>}
     }
    }
    shapedTextOptions {bOverride_TextFlowDirection=true, bOverride_TextShapingMethod=false, textFlowDirection=Culture, textShapingMethod=Auto}
   }
   slot {class=VerticalBoxSlot, properties {verticalAlignment=VAlign_Bottom}}
  }
  SizeBox_Buttons {
   widgetClassPath=&SizeBox
   bIsVariable=true
   parent=GP_Content
   widget=SizeBox_Buttons
   properties {bOverride_WidthOverride=true, widthOverride=420}
   slot {class=GridSlot, properties {column=1, row=1, verticalAlignment=VAlign_Bottom}}
  }
  VerticalBox_Buttons {
   widgetClassPath=&VerticalBox
   bIsVariable=true
   parent=SizeBox_Buttons
   widget=VerticalBox_Buttons
   properties {navigation=&self:WidgetTree.VerticalBox_Buttons.WidgetNavigation_0}
   slot {class=SizeBoxSlot, properties {verticalAlignment=VAlign_Bottom}}
  }
 ]
}
