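# MOTD page-indicator pip widget: a material-driven Border sized by a Spacer, with active state and optional timer progress.
# SetActive resets saved progress on activation and plays a forward/reverse animation fading progress opacity, also widening the pip up to double when DisplayTimer is set.
# SetTimerAnimPlaying pauses or resumes the one-second progress animation from its saved time, scaling speed by inverse duration, or pins it full without a progress bar.
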WidgetBlueprint /Motd/Widgets/WBP_MOTD_Pip_Entry
animations [
 Anim_Active {
  binding_count=2
  length_seconds=0.15
  object=&self:Anim_Active
  sequence {bindings [
    Pip {
     id=#1
     possessed_class=Border
     tracks [{
       type=MovieSceneWidgetMaterialTrack
       sections [{type=MovieSceneParameterSection, channels [Progress Opacity {type=MovieSceneScriptingFloatChannel, keys [[0,0],[3,1]]}]}]
      }]
    }
    WBP_MOTD_Pip_Entry_C_41 {
     id=#2
     possessed_class=WBP_MOTD_Pip_Entry_C
     tracks [{
       type=MovieSceneDoubleTrack
       display_name=Pip Width Ratio
       sections [{type=MovieSceneDoubleSection, end_frame=3, start_frame=0, channels [None {type=MovieSceneScriptingDoubleChannel, keys [[0,0],[3,1]]}]}]
      }]
    }
   ]}
 }
 Anim_Active_NoTimer {
  binding_count=1
  length_seconds=0.15
  object=&self:Anim_Active_NoTimer
  sequence {bindings [Pip {
     id=#1
     possessed_class=Border
     tracks [{
       type=MovieSceneWidgetMaterialTrack
       sections [{type=MovieSceneParameterSection, channels [Progress Opacity {type=MovieSceneScriptingFloatChannel, keys [[0,0],[3,1]]}]}]
      }]
    }]}
 }
 Anim_Progress {
  binding_count=1
  length_seconds=1
  object=&self:Anim_Progress
  sequence {
   bindings [Pip {
     id=#3
     possessed_class=Border
     tracks [{
       type=MovieSceneWidgetMaterialTrack
       sections [{type=MovieSceneParameterSection, channels [Progress {type=MovieSceneScriptingFloatChannel, keys [[0,0],[60,1]]}]}]
      }]
    }]
   display_rate {numerator=60}
  }
 }
]
graphs [
 EventGraph {nodes [
   n1 = SetActive!() -> (OutputDelegate, then, Active:bool=~) {is=title K2Node_CustomEvent}
   > n2 = Branch(Condition=@n1.Active)
   > Set CurrentProgressionTime(CurrentProgressionTime:float=0) -> (Output_Get:float)
   n3 = Select(
    Option 0:Widget Animation Object Reference=Get Anim_Active_NoTimer() -> (Anim_Active_NoTimer:Widget Animation Object Reference)
    Option 1:Widget Animation Object Reference=Get Anim_Active() -> (Anim_Active:Widget Animation Object Reference)
    Index=Get DisplayTimer() -> (DisplayTimer:bool)
   ) -> (ReturnValue:Widget Animation Object Reference)
   > PlayAnimationForward(InAnimation=@n3)
   PlayAnimationReverse(execute=@n2.else, InAnimation=@n3)
   Event PreConstruct()
   > SetPipSize!(execute, self, PipSize:Vector 2D Structure=Get PipSize() -> (PipSize:Vector 2D Structure)) -> (then)
  ]}
 SetDisplayTimer {nodes [
   n1 = SetDisplayTimer!() -> (then, DisplayTimer:bool=~, Show Progressbar:bool=~) {is=title K2Node_FunctionEntry}
   > Set DisplayTimer(DisplayTimer:bool=@n1) -> (Output_Get:bool)
   > Set Show Progressbar(Show Progressbar:bool=@n1.Show Progressbar) -> (Output_Get:bool)
  ]}
 SetDuration {nodes [
   n1 = SetDuration!() -> (then, DurationValue:double=1) {is=title K2Node_FunctionEntry}
   > Set Duration Value(Duration Value:double=@n1) -> (Output_Get:double)
  ]}
 SetPipSize {nodes [
   n1 = SetPipSize!() -> (then, PipSize:Vector 2D Structure) {is=title K2Node_FunctionEntry}
   > n2 = Set PipSize(PipSize:Vector 2D Structure=@n1) -> (Output_Get:Vector 2D Structure)
   > Set Size!(execute, Size:Vector 2D Structure=@n2, self:Spacer Object Reference=Get PipSpacer() -> (PipSpacer:Spacer Object Reference)) -> (then, Output_Get:Vector 2D Structure)
  ]}
 SetPipWidthRatio {nodes [
   n1 = SetPipWidthRatio!() -> (then, NewParam:double=~) {is=title K2Node_FunctionEntry}
   n2 = Get!() -> (PipSize_X:double, PipSize_Y:double)
   > Set Size(
    Size=Lerp Vector 2D!(
     A:Vector 2D Structure=Get PipSize()
     B_X:double=float * float(A=@n2, B=2) {is=title K2Node_PromotableOperator}
     B_Y:double=@n2.PipSize_Y
     Alpha:double=@n1
    ) -> (ReturnValue:Vector 2D Structure)
    self=Get PipSpacer()
   )
  ]}
 SetTimerAnimPlaying {nodes [
   n1 = SetTimerAnimPlaying!() -> (then, IsPaused:bool=~) {is=title K2Node_FunctionEntry}
   > n2 = Branch(Condition=Get Show Progressbar() -> (Show Progressbar:bool))
   > n3 = Branch(Condition=@n1)
   n4 = Get Anim_Progress() -> (Anim_Progress:Widget Animation Object Reference)
   > n5 = PauseAnimation!(execute, self:User Widget Object Reference, InAnimation:Widget Animation Object Reference=@n4) -> (then, ReturnValue:float)
   > Set CurrentProgressionTime(CurrentProgressionTime=@n5)
   PlayAnimation(
    execute=@n3.else
    InAnimation=@n4
    StartAtTime=Get CurrentProgressionTime() -> (CurrentProgressionTime:float)
    PlaybackSpeed=SafeDivide(A=1, B=Get Duration Value() -> (Duration Value:double))
   )
   PlayAnimation(execute=@n2.else, InAnimation=Get Anim_Progress(), StartAtTime=1, NumLoopsToPlay=0, PlaybackSpeed=0)
  ]}
]
variables [
 CurrentProgressionTime {type=number}
 DisplayTimer {type=boolean, default=true}
 Duration Value {type=number, default=1}
 PipSize {type=Vector2D, default=<8 8>}
 PipWidthRatio {type=number}
 Show Progressbar {type=boolean}
]
widgets {
 info {inheritedWidgetCount=0, parentClass=&UserWidget, rootWidgetClass=&Border, widgetCount=2}
 widgets [
  Pip {
   widgetClassPath=&Border
   bIsVariable=true
   widget=Pip
   properties {
    horizontalAlignment=HAlign_Left
    visibility=SelfHitTestInvisible
    background {
     bIsDynamicallyLoaded=false
     drawAs=Image
     imageType=NoImage
     mirroring=NoMirror
     resourceName=None
     resourceObject=/Motd/Widgets/Materials/MI_Pip_Entry
     tiling=NoTile
     imageSize=<8 8>
     margin=<ltrb 0.5 0.5 0.5 0.5>
     outlineSettings {
      bUseBrushTransparency=false
      roundingType=HalfHeightRadius
      width=0
      color {colorUseRule=UseColor_Specified, specifiedColor=<rgba 0 0 0 0>}
      cornerRadii=<0 0 0 0>
     }
     tintColor {colorUseRule=UseColor_Specified, specifiedColor=<rgba 1 1 1 1>}
     uVRegion {bIsValid=false, max=<0 0>, min=<0 0>}
    }
    padding=<ltrb 0 0 0 0>
   }
  }
  PipSpacer {
   widgetClassPath=&Spacer
   bIsVariable=true
   parent=Pip
   widget=PipSpacer
   properties {size=<8 8>}
   slot {class=BorderSlot, properties {horizontalAlignment=HAlign_Left, padding=<ltrb 0 0 0 0>}}
  }
 ]
}
