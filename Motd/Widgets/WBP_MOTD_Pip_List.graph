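# Widget for the MOTD screen that shows a row of pip indicators, one per item, with the active pip optionally running a timer animation.
# SetPipsCount stores the display settings, rebuilds WBP_MOTD_Pip_Entry widgets in a DynamicEntryBox, caches them in PipsArray and records the first as current.
# SetActivePip deactivates the current pip, then activates the indexed one and starts its timed animation.
# PauseActivePip pauses or resumes that animation when Anim Pip is enabled.
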WidgetBlueprint /Motd/Widgets/WBP_MOTD_Pip_List
graphs [
 EventGraph {links [], nodes []}
 PauseActivePip {nodes [
   n1 = PauseActivePip!() -> (then, Is Paused:bool=~) {is=title K2Node_FunctionEntry}
   > Branch(Condition=Get Anim Pip() -> (Anim Pip:bool))
   > n2 = Get!(execute:Exec) -> (then:Exec, else:Exec, Current Pip Entry:WBP MOTD Pip Entry Object Reference)
   > SetTimerAnimPlaying!(execute, self:WBP MOTD Pip Entry Object Reference=@n2, IsPaused:bool=@n1) -> (then)
  ]}
 SetActivePip {nodes [
   n1 = SetActivePip!() -> (then, Index:int) {is=title K2Node_FunctionEntry}
   > n2 = Get()
   > SetActive!(execute, self:WBP MOTD Pip Entry Object Reference=@n2, Active:bool=false) -> (then)
   n3 = Get PipsArray() -> (PipsArray:Array of WBP MOTD Pip Entry Object References)
   > n4 = Branch(Condition=Is Empty!(TargetArray:Array of WBP MOTD Pip Entry Object References=@n3) -> (ReturnValue:bool))
   n5 = "Get (a copy)"(Array:Array of WBP MOTD Pip Entry Object References=@n3, Dimension 1=@n1) -> (Output:WBP MOTD Pip Entry Object Reference)
   SetActive(execute=@n4.else, self=@n5, Active=true)
   > Set Current Pip Entry(Current Pip Entry:WBP MOTD Pip Entry Object Reference=@n5) -> (Output_Get:WBP MOTD Pip Entry Object Reference)
   > SetDuration!(execute, self:WBP MOTD Pip Entry Object Reference=@n5, DurationValue:double=Get Duration Value() -> (Duration Value:double)) -> (then)
   > SetTimerAnimPlaying(self=@n5, IsPaused=false)
  ]}
 SetPipsCount {nodes [
   n1 = SetPipsCount!() -> (then, Count:int, Duration Value:double=~, Anim Pip:bool=~, Show Progress:bool=~) {is=title K2Node_FunctionEntry}
   > Set Count(Count:int=@n1) -> (Output_Get:int=0)
   > Set Duration Value(Duration Value:double=@n1.Duration Value) -> (Output_Get:double)
   > Set Anim Pip(Anim Pip:bool=@n1.Anim Pip) -> (Output_Get:bool)
   > Set Show Progress(Show Progress:bool=@n1.Show Progress) -> (Output_Get:bool)
   > Clear(TargetArray:Array of WBP MOTD Pip Entry Object References=Get PipsArray())
   > Reset!(execute, self:Dynamic Entry Box Object Reference=Get PipBox() -> (PipBox:Dynamic Entry Box Object Reference), bDeleteWidgets:bool=true) -> (then)
   > n2 = For Loop(LastIndex=int - int(A=Get Count() -> (Count:int=0)))
   n3 = Get PipsArray()
   n4 = Branch(execute=@n2.Completed, Condition=Is Empty(TargetArray=@n3))
   Set Current Pip Entry(execute=@n4.else, Current Pip Entry="Get (a copy)"(Array=@n3))
   n5 = Create Entry!(execute=@n2.LoopBody, self:Dynamic Entry Box Object Reference=Get PipBox()) -> (then, ReturnValue:User Widget Object Reference)
   > n6 = Cast To WBP_MOTD_Pip_Entry!(execute, Object=@n5) -> (then, CastFailed, AsWBP MOTD Pip Entry:WBP MOTD Pip Entry Object Reference) {is=title K2Node_DynamicCast}
   > SetPipSize!(execute, self:WBP MOTD Pip Entry Object Reference=@n6, PipSize:Vector 2D Structure=Get PipSize() -> (PipSize:Vector 2D Structure)) -> (then)
   > SetDisplayTimer!(
    execute
    self:WBP MOTD Pip Entry Object Reference=@n6
    DisplayTimer:bool=Get Anim Pip()
    Show Progressbar:bool=Get Show Progress() -> (Show Progress:bool)
   ) -> (then)
   > SetActive(self=@n6, Active=false)
   > Add(TargetArray:Array of WBP MOTD Pip Entry Object References=Get PipsArray(), NewItem:"WBP MOTD Pip Entry Object Reference (by ref)"=@n6) {is=title K2Node_CallArrayFunction}
  ]}
]
variables [
 Anim Pip {type=boolean, default=true}
 Count {type=integer}
 Current Pip Entry {type=WBP_MOTD_Pip_Entry_C, default=null}
 Duration Value {type=number}
 Index {type=integer}
 PipsArray {type="array<WBP_MOTD_Pip_Entry_C>", default []}
 PipSize {type=Vector2D, default=<8 8>}
 Show Progress {type=boolean}
]
widgets {
 info {inheritedWidgetCount=0, parentClass=&UserWidget, rootWidgetClass=&DynamicEntryBox, widgetCount=1}
 widgets [PipBox {
   widgetClassPath=&DynamicEntryBox
   bIsVariable=true
   widget=PipBox
   properties {
    entryHorizontalAlignment=HAlign_Left
    entryVerticalAlignment=VAlign_Center
    entryWidgetClass=&/WBP_MOTD_Pip_Entry.WBP_MOTD_Pip_Entry_C
    numDesignerPreviewEntries=5
    visibility=HitTestInvisible
    entrySpacing=<4 0>
    radialBoxSettings {angleBetweenItems=0, bClockwiseOrder=false, bDistributeItemsEvenly=false, marginSize=0, sectorCentralAngle=360, startingAngle=0}
   }
  }]
}
