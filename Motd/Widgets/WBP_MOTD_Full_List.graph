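# Fullscreen message-of-the-day carousel widget for Fortnite: a horizontal scroll box of MOTD tiles with pip dots, arrow buttons and a close button.
# Add MOTD to List rebuilds the tiles from view models, choosing a community-post or standard tile.
# Arrows, swipes, D-pad and user scrolling move the index, which updates pips, activates that tile and reports MOTD shown and impression events.
# It also hides social text chat while constructed and adapts padding for compact UI.
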WidgetBlueprint /Motd/Widgets/WBP_MOTD_Full_List
animations [Intro {
  binding_count=1
  length_seconds=0.4
  object=&self:Intro
  sequence {
   bindings [ScrollBox {
     id=#1
     possessed_class=AthenaScrollBox
     tracks [
      {
       type=MovieSceneFloatTrack
       display_name=Render Opacity
       sections [{type=MovieSceneFloatSection, end_frame=24, channels [None {type=MovieSceneScriptingFloatChannel, keys [[0,0],[24,1]]}]}]
      }
      {
       type=MovieScene2DTransformTrack
       sections [{type=MovieScene2DTransformSection, end_frame=24, channels [Translation.Y {type=MovieSceneScriptingFloatChannel, keys [[0,64],[24,0]]}]}]
      }
     ]
    }]
   display_rate {numerator=60}
  }
 }]
graphs [
 Add MOTD to List {nodes [
   n1 = Add MOTD to List!() -> (then, Motds:Array of Motd View Model Object References) {is=title K2Node_FunctionEntry}
   > ClearMOTDList()
   > n2 = Set MyMotds(MyMotds:Array of Motd View Model Object References=@n1) -> (Output_Get:Array of Motd View Model Object References)
   > n3 = Set Count(Count:int=Length(TargetArray:Array of Motd View Model Object References=@n2)) -> (Output_Get:int=0)
   > SetPipsCount!(
    execute
    self:WBP MOTD Pip List Object Reference=Get WBP_MOTD_Pip_List() -> (WBP_MOTD_Pip_List:WBP MOTD Pip List Object Reference)
    Count:int=@n3
    Duration Value:double=0
    Anim Pip:bool=false
    Show Progress:bool=false
   ) -> (then)
   > n4 = For Each Loop(Array:Array of Motd View Model Object References=Get MyMotds() -> (MyMotds:Array of Motd View Model Object References)) -> (Array Element:Motd View Model Object Reference)
   > n5 = CreateMOTDTileBasedOnVM!(
    execute
    self
    VM:Motd View Model Object Reference="Get (a copy)"(Array:Array of Motd View Model Object References=Get MyMotds(), Dimension 1=@n4.Array Index) -> (Output:Motd View Model Object Reference)
   ) -> (then, ItemTile:WBP MOTD Full Tile Item Base Object Reference)
   > n6 = Sequence()
   > Bind Event to Navigate Left(self:WBP MOTD Full Tile Item Base Object Reference=@n5, Delegate=Create Event()) {is=title K2Node_AddDelegate}
   > Bind Event to Navigate Right(self:WBP MOTD Full Tile Item Base Object Reference=@n5, Delegate=Create Event()) {is=title K2Node_AddDelegate}
   > Bind Event to Close Widget(self:WBP MOTD Full Tile Item Base Object Reference=@n5, Delegate=Create Event()) {is=title K2Node_AddDelegate}
   SetMotd!(execute=@n6.then_1, self:Fort Dynamic Motd Fullscreen Tile Object Reference=@n5, InMotd:Motd View Model Object Reference=@n4) -> (then)
   > AddChild(self=Get ScrollBox() -> (ScrollBox:Athena Scroll Box Object Reference), Content=@n5)
  ]}
 ChildVideoPlayingControlBroadcast {nodes [
   ChildVideoPlayingControlBroadcast() {is=title K2Node_FunctionEntry}
   > Set SentManualControlEvent(SentManualControlEvent:bool=false) -> (Output_Get:bool)
   > Set ChildVideoPlaying(ChildVideoPlaying:bool=false) -> (Output_Get:bool)
   > n1 = For Each Loop with Break!(Exec:Exec, Array:Array of Widget Object References=GetAllChildren(self=Get ScrollBox()), Break:Exec=@n4.then) -> (LoopBody:Exec, Array Element:Widget Object Reference, Array Index:int, Completed:Exec)
   n2 = Branch(execute=@n1.Completed, Condition=Get ChildVideoPlaying() -> (ChildVideoPlaying:bool))
   Set SentManualControlEvent(execute=@n2.else, SentManualControlEvent=true)
   > SendManualVideoControlEvent!(execute, self:Fort Dynamic Motd Fullscreen Widget Object Reference, bEnableSecondary:bool=true) -> (then)
   n3 = Cast To WBP_MOTD_Full_TileItem!(execute=@n1.LoopBody, Object=@n1) -> (then, CastFailed, AsWBP MOTD Full Tile Item:WBP MOTD Full Tile Item Object Reference) {is=title K2Node_DynamicCast}
   > Branch(Condition=Get VideoPlaying!(self:WBP MOTD Full Tile Item Object Reference=@n3) -> (VideoPlaying:bool))
   > n4 = Set ChildVideoPlaying(ChildVideoPlaying=true)
  ]}
 ClearMOTDList {nodes [
   ClearMOTDList() {is=title K2Node_FunctionEntry}
   > n1 = For Each Loop(Array:Array of Widget Object References=GetAllChildren(self=Get ScrollBox())) -> (Array Element:Widget Object Reference)
   ClearChildren(execute=@n1.Completed, self=Get ScrollBox())
   > Clear(TargetArray:Array of Motd View Model Object References=Get MyMotds())
   n2 = Cast To WBP_MOTD_Full_TileItem(execute=@n1.LoopBody, Object=@n1)
   > Unbind Event from Navigate Left(self:WBP MOTD Full Tile Item Object Reference=@n2, Delegate=Create Event()) {is=title K2Node_RemoveDelegate}
   > Unbind Event from Navigate Right(self:WBP MOTD Full Tile Item Object Reference=@n2, Delegate=Create Event()) {is=title K2Node_RemoveDelegate}
   > DeactivateWidget(self=@n2)
  ]}
 CreateMOTDTileBasedOnVM {nodes [
   n1 = CreateMOTDTileBasedOnVM!() -> (then, VM:Motd View Model Object Reference) {is=title K2Node_FunctionEntry}
   > n2 = Branch(Condition=HasCommunityAnnouncement!(self:Motd View Model Object Reference=@n1) -> (ReturnValue:bool))
   > n3 = Create Widget!(
    execute:Exec
    Class:User Widget Class Reference=&/WBP_MOTD_CommunityPost_Full_TileItem.WBP_MOTD_CommunityPost_Full_TileItem_C
    OwningPlayer:Player Controller Object Reference
    MotdViewModel:Motd View Model Object Reference=@n1
   ) -> (then:Exec, ReturnValue:WBP MOTD Community Post Full Tile Item Object Reference)
   > Return Node!(execute, ItemTile:WBP MOTD Full Tile Item Base Object Reference=@n3)
   n4 = Create Widget(execute=@n2.else, Class=&/WBP_MOTD_Full_TileItem.WBP_MOTD_Full_TileItem_C, MotdViewModel=@n1) -> (ReturnValue:WBP MOTD Full Tile Item Object Reference)
   > Return Node(ItemTile=@n4)
  ]}
 EventGraph {nodes [
   n1 = "On Clicked (WBP_ArrowButton_Right)"!() -> (OutputDelegate, then, Button:Common Button Base Object Reference) {is=title K2Node_ComponentBoundEvent}
   Fire Event Element Interacted(
    execute=[@n1.then, @n13.then, @n19.Pressed]
    Element=Get WBP_ArrowButton_Right() -> (WBP_ArrowButton_Right:WBP Arrow Button Object Reference)
   )
   n2 = GetContentHash!(self:Motd View Model Object Reference=Get CurrentMotd() -> (CurrentMotd:Motd View Model Object Reference)) -> (ReturnValue:String)
   > OnButtonPressed!(execute, self:Fort Dynamic Motd Fullscreen Widget Object Reference, ButtonPressed:EButtonAction Enum=NextTile, ContentHash:String=@n2) -> (then)
   > Goto Right()
   n3 = "On Clicked (DecoyButton_Close)"!() -> (OutputDelegate, then, Button:Common Button Base Object Reference) {is=title K2Node_ComponentBoundEvent}
   > Fire Event Element Interacted(Element=@n3.Button)
   > DeactivateWidget()
   > OnButtonPressed(ButtonPressed=Close, ContentHash=@n2)
   n4 = "On User Scrolled (ScrollBox)"!() -> (OutputDelegate, then, CurrentOffset:float) {is=title K2Node_ComponentBoundEvent}
   > n5 = Branch(Condition=IsUsingTouch(PlayerController=GetOwningPlayer()))
   > Set IgnoreScrolled(IgnoreScrolled:bool=false) -> (Output_Get:bool)
   > Set CurrentScrollboxOffset(CurrentScrollboxOffset:double=@n4.CurrentOffset) -> (Output_Get:double)
   n6 = Branch(execute=@n5.else, Condition=NOT Boolean(A=Get IgnoreScrolled() -> (IgnoreScrolled:bool)))
   > n7 = Branch(Condition="float > float"(A=@n4.CurrentOffset, B=Get CurrentScrollboxOffset() -> (CurrentScrollboxOffset:double)))
   > n8 = Increment Int(Value=Get Index() -> (Index:int=0))
   Set CurrentScrollboxOffset(execute=[@n8."  ", @n10."  "], CurrentScrollboxOffset=@n4.CurrentOffset)
   n9 = Get Index() -> (Index=0)
   > SetActivePip!(execute, self:WBP MOTD Pip List Object Reference=Get WBP_MOTD_Pip_List(), Index:int=@n9) -> (then)
   > Set Active Slide!(execute, self, Index:int=@n9, Scrolled:bool=true) -> (then) {is=title K2Node_CallFunction}
   n10 = Decrement Int(" "=@n7.else, Value=Get Index() -> (Index=0))
   Set IgnoreScrolled(execute=@n6.else, IgnoreScrolled=false)
   > Set CurrentScrollboxOffset(CurrentScrollboxOffset=@n4.CurrentOffset)
   n11 = "On Clicked (WBP_ArrowButton_Left)"!() -> (OutputDelegate, then, Button:Common Button Base Object Reference) {is=title K2Node_ComponentBoundEvent}
   Fire Event Element Interacted(execute=[@n11.then, @n12.then, @n20.Pressed], Element=Get WBP_ArrowButton_Left() -> (WBP_ArrowButton_Left:WBP Arrow Button Object Reference))
   > OnButtonPressed(ButtonPressed=PreviousTile, ContentHash=@n2)
   > Goto Left()
   n12 = Event SwipeLeft() {is=title K2Node_Event}
   n13 = Event SwipeRight() {is=title K2Node_Event}
   Event Construct()
   n14 = Create Event()
   n15 = Get CommonInputSubsystem()
   > Bind Event to On Input Method Changed(self=@n15, Delegate=@n14)
   > HandleInputTypeChange!(execute, self, NewInputType:ECommonInputType Enum=GetCurrentInputType(self=@n15)) -> (then)
   > SetSocialTextChatVisibility!(execute, self, SetIsActive:bool=false) -> (then)
   Event Destruct()
   > Unbind Event from On Input Method Changed(self=@n15, Delegate=@n14)
   > SetSocialTextChatVisibility(SetIsActive=true)
   Event On Activated()
   > PlayAnimation(InAnimation=Get Intro())
   > SendManualVideoControlEvent(bEnableSecondary=false)
   n16 = Event SetSlideIndex!() -> (OutputDelegate, then, Index:int=0) {is=title K2Node_Event}
   > Set Index(Index:int=@n16.Index) -> (Output_Get:int=0)
   > Delay(Duration=0.2)
   n17 = Get ScrollBox()
   > ScrollWidgetIntoView!(
    execute
    self:Scroll Box Object Reference=@n17
    WidgetToFind:Widget Object Reference=GetChildAt!(self:Panel Widget Object Reference=@n17, Index:int=@n16.Index) -> (ReturnValue:Widget Object Reference)
    AnimateScroll:bool=false
    ScrollDestination:EDescendantScrollDestination Enum=Center
    Padding:float=0
   ) -> (then)
   n18 = Get WBP_MOTD_Pip_List()
   > SetPipsCount(self=@n18, Count=Get Count() -> (Count:int=0), Duration Value=0, Anim Pip=Get Anim Pip() -> (Anim Pip:bool), Show Progress=false)
   > SetActivePip(self=@n18, Index=@n16.Index)
   > Set Active Slide(Index=@n16.Index, Scrolled=false)
   Event On Deactivated()
   > ChildVideoPlayingControlBroadcast() {is=title K2Node_CallFunction}
   > Set IgnoreScrolled(IgnoreScrolled=true)
   Event PreConstruct()
   > SetCompactUIStyle() {is=title K2Node_CallFunction}
   n19 = Gamepad D-pad Right!() -> (Pressed:Exec, Released:Exec, Key:Key Structure) {is=title K2Node_InputKey, inputKey=Gamepad_DPad_Right}
   n20 = Gamepad D-pad Left!() -> (Pressed:Exec, Released:Exec, Key:Key Structure) {is=title K2Node_InputKey, inputKey=Gamepad_DPad_Left}
   HandleInputTypeChange!() -> (OutputDelegate, then, NewInputType:ECommonInputType Enum) {is=title K2Node_CustomEvent}
  ]}
 Goto Left {nodes [
   Goto Left() {is=title K2Node_FunctionEntry}
   > n1 = Branch(Condition="integer > integer"(A=Get Index() -> (Index=0)))
   > n2 = Set Index(Index=int - int(A=Get Index() -> (Index=0))) -> (Output_Get=0)
   n3 = Get ScrollBox()
   ScrollWidgetIntoView(
    execute=[@n5.then, @n2.then]
    self=@n3
    WidgetToFind=GetChildAt(self=@n3, Index=Get Index() -> (Index=0))
    AnimateScroll=true
    ScrollDestination=Center
    Padding=0
   )
   n4 = Get Index() -> (Index=0)
   > SetActivePip(self=Get WBP_MOTD_Pip_List(), Index=@n4)
   > Set Active Slide(Index=@n4, Scrolled=true)
   > Set IgnoreScrolled(IgnoreScrolled=true)
   n5 = Set Index(execute=@n1.else, Index=int - int(A=Get Count() -> (Count=0))) -> (Output_Get=0)
  ]}
 Goto Right {nodes [
   Goto Right() {is=title K2Node_FunctionEntry}
   n1 = Get Count() -> (Count=0)
   > n2 = Branch(Condition="integer > integer"(A=int - int(A=@n1), B=Get Index() -> (Index=0)))
   > n3 = Set Index(Index=int + int(A=Get Index() -> (Index=0))) -> (Output_Get=0)
   n4 = Get ScrollBox()
   ScrollWidgetIntoView(
    execute=[@n3.then, @n6.then]
    self=@n4
    WidgetToFind=GetChildAt(self=@n4, Index=Get Index() -> (Index=0))
    AnimateScroll=true
    ScrollDestination=Center
    Padding=0
   )
   n5 = Get Index() -> (Index=0)
   > SetActivePip(self=Get WBP_MOTD_Pip_List(), Index=@n5)
   > Set Active Slide(Index=@n5, Scrolled=true)
   > Set IgnoreScrolled(IgnoreScrolled=true)
   n6 = Set Index(execute=@n2.else, Index=0) -> (Output_Get=0)
  ]}
 Set Active Slide {nodes [
   n1 = Set Active Slide!() -> (then, Index:int, Scrolled:bool=~) {is=title K2Node_FunctionEntry}
   > n2 = For Loop(LastIndex=int - int(A=Get Count() -> (Count=0)))
   > n3 = Cast To WBP_MOTD_Full_TileItem_Base!(execute, Object=GetChildAt(self=Get ScrollBox(), Index=@n2)) -> (then, CastFailed, AsWBP MOTD Full Tile Item Base:WBP MOTD Full Tile Item Base Object Reference) {is=title K2Node_DynamicCast}
   > SetActive!(execute, self:WBP MOTD Full Tile Item Base Object Reference=@n3, Active:bool=false, Scroll:bool=Get Scrolled() -> (Scrolled:bool)) -> (then)
   > DeactivateWidget(self=@n3)
   n4 = Cast To WBP_MOTD_Full_TileItem_Base(execute=@n2.Completed, Object=GetChildAt(self=Get ScrollBox(), Index=@n1))
   > SetActive(self=@n4, Active=true, Scroll=Get Scrolled())
   > ActivateWidget(self=@n4)
   n5 = Get MyMotds()
   > Branch(Condition=Is Valid Index!(TargetArray:Array of Motd View Model Object References=@n5, IndexToTest:int=@n1) -> (ReturnValue:bool))
   > n6 = Set CurrentMotd(CurrentMotd:Motd View Model Object Reference="Get (a copy)"(Array=@n5, Dimension 1=@n1)) -> (Output_Get:Motd View Model Object Reference)
   > OnCurrentMotdSet!(execute, self:Fort Dynamic Motd Fullscreen Widget Object Reference, Index:int=@n1) -> (then)
   > Set Vertical Layout!(
    execute
    self:WBP MOTD Full Tile Item Base Object Reference=@n4
    IsVertical:bool=GetbIsVerticalTextLayout!(self:Motd View Model Object Reference=@n6) -> (ReturnValue:bool)
   ) -> (then) {is=title K2Node_CallFunction}
   > OnMotdShown!(execute, self:Fort Dynamic Motd Fullscreen Widget Object Reference, ContentHash:String=GetContentHash(self=@n6)) -> (then)
   n7 = Get CurrentMotd()
   > Branch(Condition=HasCommunityAnnouncement(self=@n7))
   > SendImpression!(
    execute
    self:Community Post VM Object Reference=Get CommunityAnnouncement!(self:Motd View Model Object Reference=@n7) -> (CommunityAnnouncement:Community Announcement VM Object Reference)
    LocationContext:String=MOTD
    bInteraction:bool=false
   ) -> (then)
  ]}
 SetCompactUIStyle {nodes [
   SetCompactUIStyle() {is=title K2Node_FunctionEntry}
   > Branch(Condition=IsUsingCompactUI())
   n1 = SlotAsSafeBoxSlot!(Widget:Widget Object Reference=Get Overlay_CarouselDots() -> (Overlay_CarouselDots:Overlay Object Reference)) -> (ReturnValue:Safe Zone Slot Object Reference)
   > Set VAlign!(execute, VAlign:EVerticalAlignment Enum=VAlign_Bottom, self:Safe Zone Slot Object Reference=@n1) -> (then, Output_Get:EVerticalAlignment Enum=VAlign_Fill)
   > Set!(
    execute
    Padding_Left:float=80
    Padding_Top:float=0
    Padding_Right:float=80
    Padding_Bottom:float=16
    self:Safe Zone Slot Object Reference=@n1
   ) -> (then, Output_Get:Margin Structure)
   > SetPadding!(
    execute
    self:Grid Slot Object Reference=SlotAsGridSlot(Widget=Get ScrollBox())
    InPadding_Left:float=0
    InPadding_Top:float=56
    InPadding_Right:float=0
    InPadding_Bottom:float=32
   ) -> (then)
  ]}
 SetSocialTextChatVisibility {nodes [
   n1 = SetSocialTextChatVisibility!() -> (then, SetIsActive:bool=~) {is=title K2Node_FunctionEntry}
   > Broadcast Event!(
    execute
    self:Gameplay Event Router Component Object Reference=GetGlobalEventRouter()
    EventData:BPS Set Text Chat Active State Structure=Make BPS Set Text Chat Active State!(bSetActive_5_#2:bool=@n1) -> (BPS_SetTextChatActiveState:BPS Set Text Chat Active State Structure) {is=title K2Node_MakeStruct}
    EventContext:Object Reference=GetOwningPlayer()
   ) -> (then, ReturnValue:bool)
  ]}
]
variables [
 Anim Pip {type=boolean, default=true}
 Count {type=integer}
 Current Pip Entry {type=WBP_MOTD_Pip_Entry_C, default=null}
 CurrentMotd {type=MotdViewModel, default=null}
 CurrentScrollboxOffset {type=number}
 IgnoreScrolled {type=boolean, default=true}
 Index {type=integer}
 Material {type=MaterialInstanceDynamic}
 MyMotds {type="array<MotdViewModel>", default []}
 PipsArray {type="array<WBP_MOTD_Pip_Entry_C>", default []}
 PrevPipEntry {type=WBP_MOTD_Pip_Entry_C, default=null}
 SentManualControlEvent {type=boolean}
 SwitchingMOTD {type=boolean}
]
widgets {
 info {
  inheritedWidgetCount=1
  parentClass=/Script/FortniteUI.FortDynamicMotdFullscreenWidget
  rootWidgetClass=&GridPanel
  widgetCount=12
 }
 widgets [
  GridPanel_0 {widgetClassPath=&GridPanel, widget=GridPanel_0, properties {columnFill [1], rowFill [1]}}
  WBP_UIKit_Scrim {
   widgetClassPath=/FortUILibrary/Systems/Dialog/WBP_UIKit_Scrim
   bIsVariable=true
   parent=GridPanel_0
   widget=WBP_UIKit_Scrim
   slot {class=GridSlot, properties {layer=-1}}
  }
  Tint {
   widgetClassPath=&Image
   parent=GridPanel_0
   widget=Tint
   properties {
    brush {
     bIsDynamicallyLoaded=false
     drawAs=Image
     imageType=NoImage
     mirroring=NoMirror
     resourceName=None
     resourceObject=null
     tiling=NoTile
     imageSize=<32 32>
     margin=<ltrb 0 0 0 0>
     outlineSettings {
      bUseBrushTransparency=false
      roundingType=HalfHeightRadius
      width=0
      color {colorUseRule=UseColor_Specified, specifiedColor=<rgba 0 0 0 0>}
      cornerRadii=<0 0 0 0>
     }
     tintColor {colorUseRule=UseColor_Specified, specifiedColor=<rgba 0 0 0 1>}
     uVRegion {bIsValid=false, max=<0 0>, min=<0 0>}
    }
    colorAndOpacity=<rgba 1 1 1 0.6>
   }
   slot {class=GridSlot}
  }
  GridPanel_1 {
   widgetClassPath=&GridPanel
   parent=GridPanel_0
   widget=GridPanel_1
   properties {columnFill [1], rowFill [1,0]}
   slot {class=GridSlot, properties {layer=10, padding=<ltrb 0 32 0 0>}}
  }
  ScrollBox {
   widgetClassPath=/Script/FortniteUI.AthenaScrollBox
   bInherited=true
   bIsVariable=true
   parent=GridPanel_1
   widget=ScrollBox
   properties {
    allowOverscroll=false
    bAllowRightClickDragScrolling=false
    bAnimateWheelScrolling=false
    backPadScrolling=true
    consumeMouseWheel=Never
    frontPadScrolling=true
    navigationDestination=Center
    orientation=Orient_Horizontal
    scrollBarVisibility=Collapsed
    visibility=SelfHitTestInvisible
   }
   slot {class=GridSlot, properties {verticalAlignment=VAlign_Center}}
  }
  SafeZone {
   widgetClassPath=&SafeZone
   bIsVariable=true
   parent=GridPanel_1
   widget=SafeZone
   properties {padTop=false}
   slot {class=GridSlot, properties {row=1, verticalAlignment=VAlign_Bottom}}
  }
  Overlay_CarouselDots {
   widgetClassPath=&Overlay
   bIsVariable=true
   parent=SafeZone
   widget=Overlay_CarouselDots
   slot {class=SafeZoneSlot, properties {padding=<ltrb 80 10 80 32>}}
  }
  HorizontalBox_1 {
   widgetClassPath=&HorizontalBox
   parent=Overlay_CarouselDots
   widget=HorizontalBox_1
   properties {visibility=Visible}
   slot {class=OverlaySlot, properties {horizontalAlignment=HAlign_Center, verticalAlignment=VAlign_Center}}
  }
  WBP_ArrowButton_Left {
   widgetClassPath=&/WBP_ArrowButton
   bIsVariable=true
   parent=HorizontalBox_1
   widget=WBP_ArrowButton_Left
   properties {
    bHideInputActionWithKeyboard=true
    bIsFocusable=false
    inverseArrow=true
    minWidth=48
    mobileMinHeight=88
    triggeringInputAction {dataTable=&InputActionDataTable, rowName=MOTD_NavigateLeft}
   }
   slot {class=HorizontalBoxSlot, properties {horizontalAlignment=HAlign_Center, verticalAlignment=VAlign_Center}}
  }
  WBP_MOTD_Pip_List {
   widgetClassPath=&/WBP_MOTD_Pip_List
   bIsVariable=true
   parent=HorizontalBox_1
   widget=WBP_MOTD_Pip_List
   properties {anim Pip=false, pipSize=<12 12>}
   slot {class=HorizontalBoxSlot, properties {horizontalAlignment=HAlign_Center, verticalAlignment=VAlign_Center, padding=<ltrb 32 0 32 0>}}
  }
  WBP_ArrowButton_Right {
   widgetClassPath=&WBP_ArrowButton
   bIsVariable=true
   parent=HorizontalBox_1
   widget=WBP_ArrowButton_Right
   properties {
    bHideInputActionWithKeyboard=true
    bIsFocusable=false
    minWidth=48
    mobileMinHeight=88
    triggeringInputAction {dataTable=&InputActionDataTable, rowName=MOTD_NavigateRight}
   }
   slot {class=HorizontalBoxSlot, properties {horizontalAlignment=HAlign_Center, verticalAlignment=VAlign_Center}}
  }
  DecoyButton_Close {
   widgetClassPath=/Game/UI/Frontend/BottomBar/WBP_BottomBarDecoyButton
   bIsVariable=true
   parent=GridPanel_0
   widget=DecoyButton_Close
   properties {
    bDisplayInputActionWhenNotInteractable=false
    bHideInputActionWithKeyboard=true
    triggeringInputAction {dataTable=&InputActionDataTable, rowName=Close}
   }
   slot {class=GridSlot, properties {horizontalAlignment=HAlign_Left, layer=5, verticalAlignment=VAlign_Top}}
  }
 ]
}
