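# Lobby-style MOTD teaser widget that rotates through news tiles with image or video backgrounds and opens the fullscreen MOTD list on click.
# Reset caches each tile's CMS image, then a per-tile timer advances the pips and crossfades between two material texture slots and text blocks.
# Upcoming videos preload into the idle slot, list updates with leading video are debounced, and mouse hover pauses rotation.
# Helper functions format creator-event status and interest counts.
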WidgetBlueprint /Motd/Widgets/WBP_MOTD_Teaser_List
animations [
 FrameTransition {
  binding_count=3
  length_seconds=0.3004
  object=&self:FrameTransition
  sequence {
   bindings [
    Thumbnail {
     id=#1
     possessed_class=Image
     tracks [{
       type=MovieSceneWidgetMaterialTrack
       sections [{
         type=MovieSceneParameterSection
         end_frame=18
         start_frame=0
         channels [
          Progress {type=MovieSceneScriptingFloatChannel, keys [[0,0],[18,1]]}
          Scale_A {type=MovieSceneScriptingFloatChannel, keys [[0,1],[18.026,1.05]]}
          Scale_B {type=MovieSceneScriptingFloatChannel, keys [[0,1.05],[18.026,1]]}
         ]
        }]
      }]
    }
    TeaserTextB {
     id=#2
     possessed_class=RichTextBlock
     tracks [
      {
       type=MovieSceneFloatTrack
       display_name=Render Opacity
       sections [{type=MovieSceneFloatSection, channels [None {type=MovieSceneScriptingFloatChannel, keys [[0,0],[7.455,0],[17.564,1],[18,1]]}]}]
      }
      {
       type=MovieScene2DTransformTrack
       sections [{
         type=MovieScene2DTransformSection
         channels [
          Translation.X {type=MovieSceneScriptingFloatChannel, keys [[0,0],[6,30],[18,0]]}
          Translation.Y {type=MovieSceneScriptingFloatChannel, const [0,1]}
          Angle {type=MovieSceneScriptingFloatChannel, const [0,1]}
          Scale.X {type=MovieSceneScriptingFloatChannel, const [1,1]}
          Scale.Y {type=MovieSceneScriptingFloatChannel, const [1,1]}
          Shear.X {type=MovieSceneScriptingFloatChannel, const [0,1]}
          Shear.Y {type=MovieSceneScriptingFloatChannel, const [0,1]}
         ]
        }]
      }
     ]
    }
    TeaserTextA {
     id=#3
     possessed_class=RichTextBlock
     tracks [
      {
       type=MovieSceneFloatTrack
       display_name=Render Opacity
       sections [{type=MovieSceneFloatSection, channels [None {type=MovieSceneScriptingFloatChannel, thinned_from=4, keys [[0,1],[8.717,0],[18,0]]}]}]
      }
      {
       type=MovieScene2DTransformTrack
       sections [{
         type=MovieScene2DTransformSection
         end_frame=18
         start_frame=0
         channels [
          Translation.X {type=MovieSceneScriptingFloatChannel, keys [[0,0],[17.563,-15],[18,0]]}
          Translation.Y {type=MovieSceneScriptingFloatChannel, const [0,1]}
          Angle {type=MovieSceneScriptingFloatChannel, const [0,1]}
          Scale.X {type=MovieSceneScriptingFloatChannel, const [1,1]}
          Scale.Y {type=MovieSceneScriptingFloatChannel, const [1,1]}
          Shear.X {type=MovieSceneScriptingFloatChannel, const [0,1]}
          Shear.Y {type=MovieSceneScriptingFloatChannel, const [0,1]}
         ]
        }]
      }
     ]
    }
   ]
   display_rate {numerator=60}
  }
 }
 Hover {
  binding_count=1
  length_seconds=0.6
  object=&self:Hover
  sequence {
   bindings [Thumbnail {
     id=#4
     possessed_class=Image
     tracks [{
       type=MovieSceneWidgetMaterialTrack
       sections [{
         type=MovieSceneParameterSection
         channels [
          Scale_A {type=MovieSceneScriptingFloatChannel, keys [[0,1],[36,1.1]]}
          Scale_B {type=MovieSceneScriptingFloatChannel, keys [[0,1],[36,1.1]]}
          Anim_Gradient {type=MovieSceneScriptingFloatChannel, keys [[0,0],[36,1]]}
         ]
        }]
      }]
    }]
   display_rate {numerator=60}
  }
 }
 Anim_ContentLoaded {
  binding_count=2
  length_seconds=0.2
  object=&self:Anim_ContentLoaded
  sequence {bindings [
    Thumbnail {
     id=#5
     possessed_class=Image
     tracks [{
       type=MovieSceneWidgetMaterialTrack
       sections [{
         type=MovieSceneParameterSection
         end_frame=4
         start_frame=0
         channels [Display Loading {type=MovieSceneScriptingFloatChannel, keys [[0,1],[4,0]]}]
        }]
      }]
    }
    VerticalBox_0 {
     id=#6
     possessed_class=VerticalBox
     tracks [{
       type=MovieSceneFloatTrack
       display_name=Render Opacity
       sections [{type=MovieSceneFloatSection, channels [None {type=MovieSceneScriptingFloatChannel, keys [[0,0],[4,1]]}]}]
      }]
    }
   ]}
 }
]
graphs [
 BP_GetDesiredFocusTarget {nodes [
   Get Desired Focus Target()
   > Return Node!(execute, ReturnValue:Widget Object Reference=Get Button() -> (Button:Button Object Reference))
  ]}
 BlockStartIfVideoNotReady {nodes [
   BlockStartIfVideoNotReady() {is=title K2Node_FunctionEntry}
   > Set AnyHasVideo(AnyHasVideo:bool=false) -> (Output_Get:bool)
   > n1 = For Each Loop with Break!(
    Exec:Exec
    Array:Array of Motd View Model Object References=Get MyMotds() -> (MyMotds:Array of Motd View Model Object References)
    Break:Exec=@n4.then
   ) -> (LoopBody:Exec, Array Element:Motd View Model Object Reference, Array Index:int, Completed:Exec)
   n2 = Branch(execute=@n1.Completed, Condition=Get AnyHasVideo() -> (AnyHasVideo:bool))
   > n3 = Branch(Condition=Get bVideoReady() -> (bVideoReady:bool))
   > Return Node!(execute, Block:bool=false)
   Return Node(execute=@n3.else, Block=true)
   Return Node(execute=@n2.else, Block=false)
   Branch(execute=@n1.LoopBody, Condition=HasTeaserVideoBackground!(self:Motd View Model Object Reference=@n1) -> (ReturnValue:bool))
   > n4 = Set AnyHasVideo(AnyHasVideo=true)
  ]}
 CF_CommunityAnnouncementVisibility {nodes [
   n1 = CF_CommunityAnnouncementVisibility!() -> (then, HasCommunityAnnouncement:bool=~, IsEvent:bool=~) {is=title K2Node_FunctionEntry}
   n2 = NOT Boolean(A=@n1.IsEvent)
   > Return Node!(
    execute
    Visibility:ESlateVisibility Enum="To Visibility (Boolean)"!(
     bIsVisible:bool=AND Boolean(A=@n1, B=@n2)
     TrueVisibility:ESlateVisibility Enum=SelfHitTestInvisible
     FalseVisibility:ESlateVisibility Enum=Collapsed
    ) -> (ReturnValue:ESlateVisibility Enum=Visible)
   )
  ]}
 CF_ConvertInterestedCountToText {nodes [
   n1 = CF_ConvertInterestedCountToText!() -> (then, Count:int) {is=title K2Node_FunctionEntry}
   > Return Node!(
    execute
    Text:Text=Format Text!(Format="LOCTEXT(\"{##} INTERESTED\")", "##":Text=ConvertQuantityToShorthandText!(Quantity:int=@n1) -> (Text:Text)) -> (Result)
   )
  ]}
 CF_EventHightlightVisibility {nodes [
   n1 = CF_EventHightlightVisibility!() -> (then, IsEvent:bool=~) {is=title K2Node_FunctionEntry}
   n2 = Get bEnableCreatorEvents!(self:Fort Runtime Options Object Reference=GetRuntimeOptions()) -> (bEnableCreatorEvents:bool)
   > Return Node!(
    execute
    Visibility:ESlateVisibility Enum="To Visibility (Boolean)"(bIsVisible=AND Boolean(A=@n1, B=@n2), TrueVisibility=SelfHitTestInvisible, FalseVisibility=Collapsed) -> (ReturnValue=Visible)
   )
  ]}
 CF_GetEventHighlightState {nodes [
   CF_GetEventHighlightState!() -> (then, EventStatus:ECommunityEventStatus Enum=Unknown, IsInterested:bool=~) {is=title K2Node_FunctionEntry}
   n1 = Get EventStatus() -> (EventStatus:ECommunityEventStatus Enum=Unknown)
   n2 = "Equal (Enum)"(A:ECommunityEventStatus Enum=Canceled, B:ECommunityEventStatus Enum=@n1)
   n3 = "Equal (Enum)"(A=@n1, B=Ended)
   n4 = Select(
    Option 0:E_UI_HighlightType Enum=NewEnumerator5
    Option 1:E_UI_HighlightType Enum=NewEnumerator8
    Index="Equal (Enum)"(A=Get EventStatus() -> (EventStatus=Unknown), B=Live)
   ) -> (ReturnValue:E_UI_HighlightType Enum)
   n5 = Get IsInterested() -> (IsInterested:bool)
   > Return Node!(
    execute
    ReturnValue:E_UI_HighlightType Enum=Select(Option 0=Select(Option 0=@n4, Option 1=NewEnumerator0, Index=@n5), Option 1=NewEnumerator5, Index=OR Boolean(A=@n2, B=@n3))
   )
  ]}
 CF_GetEventHightlightText {nodes [
   CF_GetEventHightlightText!() -> (then, EventStatus:ECommunityEventStatus Enum=Unknown, StartTime:Date Time Structure, EndTime:Date Time Structure, IsInterested:bool=~) {is=title K2Node_FunctionEntry}
   n1 = Get EventStatus() -> (EventStatus=Unknown)
   > n2 = Branch(Condition=OR Boolean(A="Equal (Enum)"(A=Ended, B=@n1), B="Equal (Enum)"(A=@n1, B=Canceled)))
   n3 = DateTime - DateTime!(A:Date Time Structure=UTC Now!() -> (ReturnValue:Date Time Structure), B:Date Time Structure=Get EndTime() -> (EndTime:Date Time Structure)) -> (ReturnValue:Timespan Structure)
   n4 = "Absolute (Integer)"!(A:int=Round(A=Get Total Days!(A:Timespan Structure=@n3) -> (ReturnValue:double) {is=title K2Node_CallFunction}) {is=title K2Node_CallFunction}) -> (ReturnValue:int=0)
   n5 = "Absolute (Integer)"(A=Floor(A=Get Total Hours!(A:Timespan Structure=@n3) -> (ReturnValue:double) {is=title K2Node_CallFunction}) {is=title K2Node_CallFunction}) -> (ReturnValue=0)
   n6 = "Absolute (Integer)"(A=Get Minutes!(A:Timespan Structure=@n3) -> (ReturnValue:int=0) {is=title K2Node_CallFunction}) -> (ReturnValue=0)
   n7 = SelectText(
    A=Format Text!(Format="LOCTEXT(\"Ended {Days} {Days}|plural(one=DAY,other=DAYS) ago\")", Days:int=@n4) -> (Result)
    B=SelectText(
     A=Format Text!(Format="LOCTEXT(\"Ended {Time} {Time}|plural(one=MINUTE,other=MINUTES) ago\")", Time:int=@n6) -> (Result)
     B=Format Text(Format="LOCTEXT(\"Ended {Time} {Time}|plural(one=HOUR,other=HOURS) ago\")", Time=@n5)
     bPickA="integer <= integer"(A=@n5)
    )
    bPickA="integer > integer"(A=@n4)
   )
   n8 = "Absolute (Integer)"(A=Floor(A=Get Total Days(A=@n3))) -> (ReturnValue=0)
   n9 = Format Text!(Format="LOCTEXT(\"Ended {Months} {Months}|plural(one=MONTH,other=MONTHS) ago\")", Months:int=@n8) -> (Result)
   n10 = "integer > integer"(A=int / int(A=@n8, B=30))
   n11 = DateTime - DateTime(A=Get StartTime() -> (StartTime:Date Time Structure), B=UTC Now())
   n12 = "Absolute (Integer)"(A=Round(A=Get Total Days(A=@n11))) -> (ReturnValue=0)
   n13 = "Absolute (Integer)"(A=Floor(A=Get Total Hours(A=@n11))) -> (ReturnValue=0)
   n14 = "Absolute (Integer)"(A=Get Minutes(A=@n11) -> (ReturnValue=0)) -> (ReturnValue=0)
   n15 = SelectText(
    A=Format Text!(Format="LOCTEXT(\"IN {Days} {Days}|plural(one=DAY,other=DAYS)\")", Days:int=@n12) -> (Result)
    B=SelectText(
     A=Format Text!(Format="LOCTEXT(\"IN {Time} {Time}|plural(one=MINUTE,other=MINUTES)\")", Time:int=@n14) -> (Result)
     B=Format Text(Format="LOCTEXT(\"IN {Time} {Time}|plural(one=HOUR,other=HOURS)\")", Time=@n13)
     bPickA="integer <= integer"(A=@n13)
    )
    bPickA="integer > integer"(A=@n12)
   )
   n16 = "Absolute (Integer)"(A=Floor(A=Get Total Days(A=@n11))) -> (ReturnValue=0)
   n17 = Format Text!(Format="LOCTEXT(\"IN {Months} {Months}|plural(one=MONTH,other=MONTHS)\")", Months:int=@n16) -> (Result)
   n18 = "integer > integer"(A=int / int(A=@n16, B=30))
   Return Node!(
    execute=[@n2.then, @n19.then]
    Text:Text=Select!(
     Unknown:Text
     Upcoming:Text=SelectText(A=@n17, B=@n15, bPickA=@n18)
     Live:Text="LOCTEXT(\"Live now\")"
     Ended:Text=SelectText(A=@n9, B=@n7, bPickA=@n10)
     Canceled:Text="LOCTEXT(\"Canceled\")"
     Index:ECommunityEventStatus Enum=Get EventStatus() -> (EventStatus=Unknown)
    ) -> (ReturnValue:Text)
   )
   n19 = Branch(execute=@n2.else, Condition=NOT Boolean(A=Get IsInterested()))
   Return Node(execute=@n19.else, Text="LOCTEXT(\"Interested\")")
  ]}
 CF_GetInterestedTextVisibility {nodes [
   n1 = CF_GetInterestedTextVisibility!() -> (then, Count:int, IsEvent:bool=~) {is=title K2Node_FunctionEntry}
   n2 = "integer > integer"(A=@n1)
   n3 = Get bEnableCreatorEvents(self=GetRuntimeOptions())
   > Return Node!(
    execute
    Visibility:ESlateVisibility Enum="To Visibility (Boolean)"(bIsVisible=AND Boolean!(A=@n2, B=@n1.IsEvent, C:bool=@n3) -> (ReturnValue), TrueVisibility=SelfHitTestInvisible, FalseVisibility=Collapsed) -> (ReturnValue=Visible)
   )
  ]}
 ConvertReactionCountToVisibility {nodes [
   n1 = ConvertReactionCountToVisibility!() -> (then, Count:int) {is=title K2Node_FunctionEntry}
   > Return Node(Visibility=Select!(Option 0:ESlateVisibility Enum=Collapsed, Option 1:ESlateVisibility Enum=SelfHitTestInvisible, Index:bool="integer > integer"(A=@n1)) -> (ReturnValue:ESlateVisibility Enum))
  ]}
 DebouncedUpdateMOTD {nodes [
   DebouncedUpdateMOTD() {is=title K2Node_FunctionEntry}
   > n1 = Set Count(Count:int=Length(TargetArray:Array of Motd View Model Object References=Get MyMotds())) -> (Output_Get:int=0)
   > n2 = Branch(Condition="integer > integer"(A=@n1))
   > Reset MOTD State!(execute, self, FromActivated:bool=false) -> (then)
   > n3 = Branch(Condition="integer > integer"(A=Get Count() -> (Count:int=0), B=1))
   > SetVisibility(self=Get WBP_MOTD_Pip_List() -> (WBP_MOTD_Pip_List:WBP MOTD Pip List Object Reference), InVisibility=SelfHitTestInvisible)
   > SetActivePip!(execute, self:WBP MOTD Pip List Object Reference=Get WBP_MOTD_Pip_List(), Index:int=0) -> (then)
   SetVisibility(execute=@n3.else, self=Get WBP_MOTD_Pip_List(), InVisibility=Collapsed)
   SetVisibility(execute=@n2.else, self=Get Overlay_MOTD() -> (Overlay_MOTD:Overlay Object Reference), InVisibility=Collapsed)
   > n4 = Construct Object from Class(Class=/Script/FortniteUI.MotdViewModel) -> (ReturnValue:Motd View Model Object Reference)
   > Set MotdViewModel(MotdViewModel:Motd View Model Object Reference=@n4) -> (Output_Get:Motd View Model Object Reference)
  ]}
 EventGraph {nodes [
   "On Hovered (Button)"() {is=title K2Node_ComponentBoundEvent}
   > SpawnSound2D!(
    execute
    Sound:Sound Base Object Reference=Get HoverSound() -> (HoverSound:Sound Base Object Reference)
    VolumeMultiplier:float=1
    PitchMultiplier:float=1
    StartTime:float=0
    ConcurrencySettings:Sound Concurrency Object Reference
    bPersistAcrossLevelTransition:bool=false
    bAutoDestroy:bool=true
   ) -> (then, ReturnValue:Audio Component Object Reference)
   n1 = IsUsingTouch(PlayerController=GetOwningPlayer())
   > n2 = Branch(Condition=@n1)
   n3 = Get Timer Handle() -> (Timer Handle:Timer Handle Structure)
   Pause Timer by Handle!(execute=@n2.else, Handle:Timer Handle Structure=@n3) -> (then)
   > PlayAnimation(InAnimation=Get Hover() -> (Hover:Widget Animation Object Reference))
   n4 = Get Block_Outline() -> (Block_Outline:WBP UIKit Block Outline Object Reference)
   > RequestTransition!(execute, self:UIKit Block Object Reference=@n4, State:EUIKitBlockVisualState Enum=Focused) -> (then)
   > PauseActivePip!(execute, self:WBP MOTD Pip List Object Reference=Get WBP_MOTD_Pip_List(), Is Paused:bool=true) -> (then)
   "On Unhovered (Button)"() {is=title K2Node_ComponentBoundEvent}
   > n5 = Branch(Condition=@n1)
   Unpause Timer by Handle!(execute=@n5.else, Handle:Timer Handle Structure=@n3) -> (then)
   > PlayAnimation(InAnimation=Get Hover(), PlayMode=Reverse, PlaybackSpeed=4)
   > RequestTransition(self=@n4, State=Unfocused)
   > PauseActivePip(self=Get WBP_MOTD_Pip_List(), Is Paused=false)
   n6 = "On Clicked (Button)"() {is=title K2Node_ComponentBoundEvent}
   n7 = Clear and Invalidate Timer by Handle(execute=[@n6.then, @n9.then], Handle=Get Timer Handle())
   n8 = FUllScreenWidgetCreatedEvent!() -> (OutputDelegate, then, Widget:Fort Dynamic Motd Fullscreen Widget Object Reference) {is=title K2Node_CustomEvent}
   Bind Event to Fullscreen Widget Created(execute=@n7.then, self:Fort Dynamic Motd Teaser Widget Object Reference, Delegate=@n8) {is=title K2Node_AddDelegate}
   > AddFullScreenWidget!(
    execute
    self:Fort Dynamic Motd Teaser Widget Object Reference
    FullScreenWidget:Fort Dynamic Motd Fullscreen Widget Soft Class Reference=&/WBP_MOTD_Full_List.WBP_MOTD_Full_List_C
    Index:int=Get Index() -> (Index:int=0)
   ) -> (then)
   > SpawnSound2D(
    Sound=Get PressedSound() -> (PressedSound:Sound Base Object Reference)
    VolumeMultiplier=1
    PitchMultiplier=1
    StartTime=0
    bPersistAcrossLevelTransition=false
    bAutoDestroy=true
   )
   n9 = "On Clicked (ActionButton)"!() -> (OutputDelegate, then, Button:Common Button Base Object Reference) {is=title K2Node_ComponentBoundEvent}
   n10 = Start Next Frame!() -> (OutputDelegate, then, FromLoop:bool=~) {is=title K2Node_CustomEvent}
   n11 = GetIndex!(execute=[@n10.then, @n40.then], self, WantIndex:int=Get Index() -> (Index=0), Increment:bool=@n10.FromLoop) -> (then, Index:int=0)
   > Set Index(Index:int=@n11) -> (Output_Get:int=0)
   > Set IsFirstLoop(IsFirstLoop:bool=NOT Boolean(A=@n10.FromLoop)) -> (Output_Get:bool)
   > SetCurrentMotd!(execute, self:Fort Dynamic Motd Teaser Widget Object Reference, Index:int=Get Index() -> (Index=0)) -> (then)
   > StopVideo()
   n12 = NOT Boolean(A=@n10.FromLoop)
   n13 = GetNextTileInfo!(First:bool=@n12, Peek:bool=false) -> (Param:Name=None, TextBlock:Rich Text Block Object Reference)
   > n14 = Set MOTD!(
    execute
    self
    Index:int=Get Index() -> (Index=0)
    TextureParam:Name=@n13
    TeaserText:Rich Text Block Object Reference=@n13.TextBlock
    First:bool=@n12
   ) -> (then, OutTextureParam:Name=None) {is=title K2Node_CallFunction}
   > n15 = Branch(Condition=NOT Boolean(A=@n10.FromLoop))
   > n16 = OnMotdShown(self:Fort Dynamic Motd Teaser Widget Object Reference)
   SetActivePip(execute=[@n16.then, @n17.then], self=Get WBP_MOTD_Pip_List(), Index=Get Index() -> (Index=0))
   n17 = "Play Animation with Finished event (legacy)"!(
    execute:Exec=@n15.else
    Widget:User Widget Object Reference=Self-Reference()
    InAnimation:Widget Animation Object Reference=Get FrameTransition() -> (FrameTransition:Widget Animation Object Reference)
    StartAtTime:float=0
    NumLoopsToPlay:int=1
    PlayMode:EUMGSequencePlayMode Enum=Select(Option 0:EUMGSequencePlayMode Enum, Option 1:EUMGSequencePlayMode Enum=Reverse, Index="Equal (Name)"(A=@n14, B=Image_A)) -> (ReturnValue:EUMGSequencePlayMode Enum)
    PlaybackSpeed:float=1
   ) -> (then:Exec, Result:UMGSequence Player Object Reference, Finished:Exec) {is=title K2Node_PlayAnimation}
   OnMotdShown(execute=@n17.Finished)
   StopVideo() {is=title K2Node_CustomEvent}
   n18 = Get CurrentTeaser() -> (CurrentTeaser:Text)
   > n19 = Branch(Condition=TextIsEmpty(InText=@n18))
   ReleaseMediaPlayer!(execute=@n19.else, self:Fort Dynamic Motd Teaser Widget Object Reference, PlayerToRemove:"Text (by ref)"=@n18) -> (then)
   > Set CurrentTeaser(CurrentTeaser:Text) -> (Output_Get:Text)
   n20 = Add Progress() {is=title K2Node_CustomEvent}
   > Start Next Frame!(execute, self, FromLoop:bool=true) -> (then) {is=title K2Node_CallFunction}
   n21 = Start Timer!() -> (OutputDelegate, then, Time:double=~) {is=title K2Node_CustomEvent}
   > n22 = Set Timer by Event(Delegate=@n20, Time=@n21.Time, bLooping=false)
   > Set Timer Handle(Timer Handle:Timer Handle Structure=@n22) -> (Output_Get:Timer Handle Structure)
   Add MOTD Images to Array() {is=title K2Node_CustomEvent}
   n23 = Get Motds!(self:Motd List View Model Object Reference=Get MotdListViewModel() -> (MotdListViewModel:Motd List View Model Object Reference)) -> (Motds:Array of Motd View Model Object References)
   n24 = Get TeaserImages() -> (TeaserImages:Array of Texture 2D Object References)
   > n25 = Branch(Condition="integer > integer"(A=Length(TargetArray=@n23), B=Length(TargetArray:Array of Texture 2D Object References=@n24)))
   > Get CMS Image!(execute, self, Image Index:int=Length(TargetArray=Get TeaserImages())) -> (then) {is=title K2Node_CallFunction}
   n26 = Branch(execute=@n25.else, Condition=Get bVideoReady())
   Unbind all Events from Video Player Ready!(execute:Exec=[@n26.then, @n46.then], self:Fort Dynamic Motd Teaser Widget Object Reference) -> (then:Exec) {is=title K2Node_ClearDelegate}
   > SetLoadingDisplay!(execute, self, IsLoading:bool=false) -> (then)
   > Start Next Frame(FromLoop=false)
   n27 = VideoReadyEvent!() -> (OutputDelegate, then, bReady:bool=~) {is=title K2Node_CustomEvent}
   Bind Event to Video Player Ready(execute=@n26.else, self:Fort Dynamic Motd Teaser Widget Object Reference, Delegate=@n27) {is=title K2Node_AddDelegate}
   OnImageLoadingComplete() {is=title K2Node_CustomEvent}
   n28 = Get Image!(self:CMSImage Object Reference=Get CMS Image() -> (CMS Image:CMSImage Object Reference) {is=title K2Node_VariableGet}) -> (Image:Epic CMSImage Object Reference)
   > Add(
    TargetArray:Array of Texture 2D Object References=Get TeaserImages()
    NewItem:"Texture 2D Object Reference (by ref)"=GetBrushResourceAsTexture2D!(Brush:"Slate Brush Structure (by ref)"=Get Brush!(self:Image Object Reference=@n28) -> (Brush:Slate Brush Structure)) -> (ReturnValue:Texture 2D Object Reference)
   ) {is=title K2Node_CallArrayFunction}
   > Add MOTD Images to Array() {is=title K2Node_CallFunction}
   n29 = FullScreenWidgetClosed() {is=title K2Node_CustomEvent}
   > Set FullScreenMotdActioning(FullScreenMotdActioning:bool=false) -> (Output_Get:bool)
   > n30 = For Each Loop with Break(
    Array:Array of Widget Object References=GetAllChildren(self=Get ScrollBox!(self:Fort Dynamic Motd Fullscreen Widget Object Reference=@n8.Widget) -> (ScrollBox:Athena Scroll Box Object Reference))
    Break=@n35.then
   ) -> (Array Element:Widget Object Reference)
   n31 = Cast To WBP_MOTD_Full_List!(execute=@n30.Completed, Object=@n8.Widget) -> (then, CastFailed, AsWBP MOTD Full List:WBP MOTD Full List Object Reference) {is=title K2Node_DynamicCast}
   n32 = Get FullScreenMotdActioning() -> (FullScreenMotdActioning:bool)
   n33 = Get SentManualControlEvent!(self:WBP MOTD Full List Object Reference=@n31) -> (SentManualControlEvent:bool)
   > Branch(Condition=AND Boolean!(A=NOT Boolean(A=@n32), B=NOT Boolean(A=@n33)) -> (ReturnValue))
   > Reset MOTD State(FromActivated=false)
   n34 = Cast To WBP_MOTD_Full_TileItem!(execute=@n30.LoopBody, Object=@n30) -> (then, CastFailed, AsWBP MOTD Full Tile Item:WBP MOTD Full Tile Item Object Reference) {is=title K2Node_DynamicCast}
   > Branch(Condition=Get DoingAction!(self:WBP MOTD Full Tile Item Object Reference=@n34) -> (DoingAction:bool))
   > n35 = Set FullScreenMotdActioning(FullScreenMotdActioning=true)
   n36 = StartVideoBP!() -> (OutputDelegate, then, Motd:"Motd View Model Object Reference (by ref)", TextureParam:Name, PlayImmediately:bool=~) {is=title K2Node_CustomEvent}
   > n37 = StartVideo!(
    execute
    self:Fort Dynamic Motd Teaser Widget Object Reference
    Motd:Motd View Model Object Reference=@n36.Motd
    InTexture:Media Texture Object Reference=Select(
     Option 0:Media Texture Object Reference=/Motd/Widgets/Textures/MT_DynamicMOTD_B
     Option 1:Media Texture Object Reference=/Motd/Widgets/Textures/MT_DynamicMOTD
     Index="Equal (Name)"(A=@n36.TextureParam, B=Image_A)
    ) -> (ReturnValue:Media Texture Object Reference)
    InTextureParam:Name=@n36.TextureParam
    bPlayImmediately:bool=@n36.PlayImmediately
   ) -> (then, ReturnValue:Dynamic Motd Video Player State Object Reference)
   > Branch(Condition=IsValid(Object=@n37))
   > Bind Event to Playing Event(self:Dynamic Motd Video Player State Object Reference=@n37, Delegate=Create Event()) {is=title K2Node_AddDelegate}
   > Bind Event to Ended Event(self:Dynamic Motd Video Player State Object Reference=@n37, Delegate=Create Event()) {is=title K2Node_AddDelegate}
   Event PreConstruct()
   > SetCompactUIStyle() {is=title K2Node_CallFunction}
   n38 = Get EventHighlight() -> (EventHighlight:WBP UIKit Highlight Object Reference)
   > SetVisibility(self=@n38, InVisibility=Collapsed)
   > SetFontSize!(
    execute
    self:Text Object Reference=Get Text_Highlight!(self:WBP UIKit Highlight Base Object Reference=@n38) -> (Text_Highlight:Common Text Object Reference)
    DisplayFontSize:float=16
   ) -> (then)
   Event Construct()
   > n39 = Create Widget!(execute:Exec, Class:User Widget Class Reference=/Game/UI/Foundation/CMS/CMSImage.CMSImage_C, OwningPlayer:Player Controller Object Reference) -> (then:Exec, ReturnValue:CMSImage Object Reference)
   > Set CMS Image(CMS Image:CMSImage Object Reference=@n39) -> (Output_Get:CMSImage Object Reference)
   > SetVisibility(self=Get EventHighlight(), InVisibility=Collapsed)
   n40 = Event OnVideoBackgroundsAllowedChanged!() -> (OutputDelegate, then, bAllowVideoBackgrounds:bool) {is=title K2Node_Event}
   Event On Activated()
   > Branch(Condition=Get bShouldBeActive() -> (bShouldBeActive:bool))
   > n41 = Branch(Condition=Get ActivatedOnce() -> (ActivatedOnce:bool))
   > n42 = Reset MOTD State(FromActivated=true)
   SetBindings!(execute=[@n42.then, @n44.then], self, Bindings Enabled:bool=true) -> (then)
   > n43 = Get!(execute:Exec) -> (then:Exec, else:Exec, FortInputStackVM:Fort Input Stack VM Object Reference)
   > PushInputContext!(
    execute
    self:Fort Input Stack VM Object Reference=@n43
    Target:Fort Input Stack Interface Interface=Get ActionButton() -> (ActionButton:WBP UIKit Button Generic Object Reference)
    InputActionRow:"Data Table Row Handle Structure (by ref)"=Get ActionInputRow() -> (ActionInputRow:Data Table Row Handle Structure)
   ) -> (then)
   n44 = Set ActivatedOnce(execute=@n41.else, ActivatedOnce:bool=true) -> (Output_Get:bool)
   Event On Deactivated()
   > Clear and Invalidate Timer by Handle(Handle=Get Timer Handle())
   > SetBindings(Bindings Enabled=false)
   > Set CurrentTeaser()
   Bind Event to On Widget Deactivated(execute=@n47.then, self:Common Activatable Widget Object Reference=@n8.Widget, Delegate=@n29) {is=title K2Node_AddDelegate}
   n45 = IsValid(Object=@n8.Widget)
   n46 = Branch(execute=@n27.then, Condition=@n27.bReady)
   n47 = Branch(execute=@n8.then, Condition=@n45)
  ]}
 Get CMS Image {nodes [
   n1 = Get CMS Image!() -> (then, Image Index:int) {is=title K2Node_FunctionEntry}
   n2 = Get CMS Image() {is=title K2Node_VariableGet}
   > Bind Event to On Image Loading Complete(self:Epic CMSImage Object Reference=Get Image(self=@n2), Delegate=Create Event()) {is=title K2Node_AddDelegate}
   n3 = Get Motds(self=Get MotdListViewModel())
   n4 = Break Motd Background!(MotdBackground:"Motd Background Structure (by ref)"=Get TeaserBackground!(self:Motd View Model Object Reference="Get (a copy)"(Array:Array of Motd View Model Object References=@n3, Dimension 1=@n1) -> (Output:Motd View Model Object Reference)) -> (TeaserBackground:Motd Background Structure)) -> (Images:Array of Motd Image Structures) {is=title K2Node_BreakStruct}
   > SetMediaURL!(
    execute
    self:CMSImage Object Reference=@n2
    MediaUrl:String=Break Motd Image!(MotdImage:"Motd Image Structure (by ref)"="Get (a copy)"(Array:Array of Motd Image Structures=@n4) -> (Output:Motd Image Structure)) -> (Url:String) {is=title K2Node_BreakStruct}
   ) -> (then)
  ]}
 Get MOTD View Models Array {nodes [
   n1 = Get MOTD View Models Array!() -> (then, Motds:Array of Motd View Model Object References) {is=title K2Node_FunctionEntry}
   > Set MyMotds(MyMotds:Array of Motd View Model Object References=@n1) -> (Output_Get:Array of Motd View Model Object References)
   > n2 = Branch(Condition="integer > integer"(A=Length(TargetArray:Array of Motd View Model Object References=@n1)))
   > n3 = Branch(Condition=HasTeaserVideoBackground(self="Get (a copy)"(Array:Array of Motd View Model Object References=@n1) -> (Output:Motd View Model Object Reference)))
   > Clear and Invalidate Timer by Handle(Handle=Get DebounceUpdateHandle() -> (DebounceUpdateHandle:Timer Handle Structure))
   > n4 = GetMotdUpdateDebounce!(execute, self:Fort Dynamic Motd Teaser Widget Object Reference) -> (then, ReturnValue:float)
   > n5 = Set Timer by Event(Delegate=Create Event(), Time=@n4, bLooping=false)
   > Set DebounceUpdateHandle(DebounceUpdateHandle:Timer Handle Structure=@n5) -> (Output_Get:Timer Handle Structure)
   > ReleaseMediaPlayer(PlayerToRemove=GetEmptyText!() -> (ReturnValue:Text))
   DebouncedUpdateMOTD(execute=[@n2.else, @n3.else]) {is=title K2Node_CallFunction}
  ]}
 GetIndex {nodes [
   n1 = GetIndex!() -> (then, WantIndex:int, Increment:bool=~) {is=title K2Node_FunctionEntry}
   > n2 = Branch(Condition="Equal (Integer)"(A=@n1, B=Get Count() -> (Count=0)))
   > Return Node!(execute, Index:int=0)
   n3 = Branch(execute=@n2.else, Condition=Get Increment() -> (Increment:bool))
   > n4 = Increment Int(Value=Get WantIndex() -> (WantIndex:int=0))
   > n5 = GetIndex(WantIndex=@n4, Increment=false) -> (Index=0) {is=title K2Node_CallFunction}
   > Return Node(Index=@n5)
   Return Node(execute=@n3.else, Index=Get WantIndex() -> (WantIndex=0))
  ]}
 GetNextTileInfo {nodes [
   n1 = GetNextTileInfo!() -> (then, First:bool=~, Peek:bool=~) {is=title K2Node_FunctionEntry}
   > n2 = Branch(Condition=@n1)
   > Set ImageA(ImageA:bool=true) -> (Output_Get:bool)
   > Return Node!(
    execute
    Param:Name=Image_A
    TextBlock:Rich Text Block Object Reference=Get TeaserTextA() -> (TeaserTextA:Common Rich Text Block Object Reference)
   )
   n3 = Branch(execute=@n2.else, Condition=Get ImageA() -> (ImageA:bool))
   > n4 = Branch(Condition=Get Peek() -> (Peek:bool))
   Return Node(execute=[@n4.then, @n5.then], Param=Image_B, TextBlock=Get TeaserTextB() -> (TeaserTextB:Common Rich Text Block Object Reference))
   n5 = Set ImageA(execute=@n4.else, ImageA=false)
   n6 = Branch(execute=@n3.else, Condition=Get Peek())
   Return Node(execute=[@n6.then, @n7.then], Param=Image_A, TextBlock=Get TeaserTextA())
   n7 = Set ImageA(execute=@n6.else, ImageA=true)
  ]}
 Reset MOTD State {nodes [
   n1 = Reset MOTD State!() -> (then, FromActivated:bool=~) {is=title K2Node_FunctionEntry}
   > n2 = Branch(Condition=@n1)
   n3 = Branch(execute=[@n7.then, @n2.then], Condition=Is Timer Active by Handle!(Handle:Timer Handle Structure=Get Timer Handle()) -> (ReturnValue:bool))
   > n4 = Clear and Invalidate Timer by Handle(Handle=Get Timer Handle())
   Clear(execute=[@n4.then, @n3.else], TargetArray:Array of Texture 2D Object References=Get TeaserImages())
   > Set Index(Index=0) -> (Output_Get=0)
   > SetLoadingDisplay(IsLoading=true)
   > Set ImageA(ImageA=true)
   > n5 = GetDynamicMaterial(self=Get Thumbnail() -> (Thumbnail:Image Object Reference))
   > SetScalarParameterValue(self=@n5, ParameterName=Progress, Value=0)
   > SetPipsCount!(
    execute
    self:WBP MOTD Pip List Object Reference=Get WBP_MOTD_Pip_List()
    Count:int=Get Count() -> (Count=0)
    Duration Value:double=Get Entry Show Time() -> (Entry Show Time:double)
    Anim Pip:bool=Get Animate Pip() -> (Animate Pip:bool)
    Show Progress:bool=Get Show Progressbar() -> (Show Progressbar:bool)
   ) -> (then)
   n6 = Get MyMotds()
   > Branch(Condition="integer > integer"(A=Length(TargetArray=@n6)))
   > Add MOTD Images to Array()
   > SetVisibility(self=Get Overlay_MOTD(), InVisibility=SelfHitTestInvisible)
   > SetMotdViewModel!(execute, self, Viewmodel:Motd View Model Object Reference) -> (then)
   n7 = ReleaseMediaPlayer(execute=@n2.else, PlayerToRemove=GetEmptyText())
  ]}
 Set MOTD {nodes [
   Set MOTD!() -> (then, Index:int, TextureParam:Name, TeaserText:Rich Text Block Object Reference, First:bool=~) {is=title K2Node_FunctionEntry}
   n1 = "Get (a copy)"(Array=Get MyMotds(), Dimension 1=Get Index() -> (Index=0))
   > SetMotdViewModel(Viewmodel=@n1)
   > Branch(Condition=IsValid(Object=@n1))
   > n2 = Get!(execute, self:Motd View Model Object Reference=@n1) -> (then, else, CommunityAnnouncement:Community Announcement VM Object Reference)
   Set Text(
    execute=[@n2.then, @n15.then]
    Text=Get TeaserTitle!(self:Motd View Model Object Reference=Get MotdViewModel() -> (MotdViewModel:Motd View Model Object Reference)) -> (TeaserTitle:Text)
    self:Rich Text Block Object Reference=Get TeaserText() -> (TeaserText:Rich Text Block Object Reference)
   )
   > n3 = GetDynamicMaterial(self=Get Thumbnail())
   > SetTextureParameterValue(
    self=@n3
    ParameterName=Get TextureParam() -> (TextureParam:Name=None)
    Value="Get (a copy)"(Array:Array of Texture 2D Object References=Get TeaserImages(), Dimension 1=Get Index() -> (Index=0)) -> (Output:Texture 2D Object Reference)
   )
   > n4 = Branch(Condition=HasTeaserVideoBackground(self=Get MotdViewModel()))
   > n5 = StartVideoBP!(
    execute
    self
    Motd:"Motd View Model Object Reference (by ref)"=Get MotdViewModel()
    TextureParam:Name=Get TextureParam() -> (TextureParam=None)
    PlayImmediately:bool=Get First() -> (First:bool)
   ) -> (then) {is=title K2Node_CallFunction}
   n6 = Get CurrentMotd() -> (CurrentMotd:Motd View Model Object Reference)
   Start Timer!(
    execute=[@n14.then, @n5.then]
    self
    Time:double="To Float (Integer)"(InInt=Get Duration!(self:Motd View Model Object Reference=@n6) -> (Duration:int=0))
   ) -> (then) {is=title K2Node_CallFunction}
   > n7 = GetIndex(WantIndex=Get Index() -> (Index=0), Increment=true) -> (Index=0)
   n8 = "Get (a copy)"(Array:Array of Motd View Model Object References=Get MyMotds(), Dimension 1=@n7) -> (Output:Motd View Model Object Reference)
   > n9 = Branch(Condition=AND Boolean(A=IsValid(Object=@n8), B=HasTeaserVideoBackground(self=@n8)))
   > n10 = StartVideoBP(Motd=@n8, TextureParam=GetNextTileInfo(First=false, Peek=true) -> (Param=None) {is=title K2Node_CallFunction}, PlayImmediately=false)
   n11 = Branch(execute=[@n10.then, @n9.else], Condition=Get First())
   n12 = Get TextureParam() -> (TextureParam=None)
   > n13 = SetRenderOpacity(
    self=Select(
     Option 0:Widget Object Reference=Get TeaserTextB()
     Option 1:Widget Object Reference=Get TeaserTextA()
     Index="Equal (Name)"(A=@n12, B=Image_A)
    ) -> (ReturnValue:Widget Object Reference)
    InOpacity=1
   )
   Return Node!(execute=[@n13.then, @n11.else], OutTextureParam:Name=Get TextureParam() -> (TextureParam=None))
   SetLoadingDisplay(execute=@n4.else, IsLoading=false)
   > n14 = SetRenderOpacity(self=Get TeaserText(), InOpacity=1)
   n15 = SetVisibility(
    execute=@n2.else
    self=[Get EventHighlight(), Get Text_Interested() -> (Text_Interested:Common Text Object Reference)]
    InVisibility=Collapsed
   )
  ]}
 SetBindings {nodes [
   n1 = SetBindings!() -> (then, Bindings Enabled:bool=~) {is=title K2Node_FunctionEntry}
   > n2 = Branch(Condition=@n1)
   n3 = Get InputActionWidget() -> (InputActionWidget:Common Action Widget Object Reference)
   > SetInputAction!(
    execute
    self:Common Action Widget Object Reference=@n3
    InputActionRow:Data Table Row Handle Structure=Make Data Table Row Handle!(DataTable:Data Table Object Reference=&InputActionDataTable, RowName:Name=GenericLeftTrigger) -> (DataTableRowHandle:Data Table Row Handle Structure) {is=title K2Node_MakeStruct}
   ) -> (then)
   SetInputAction(execute=@n2.else, self=@n3)
  ]}
 SetCompactUIStyle {nodes [
   SetCompactUIStyle() {is=title K2Node_FunctionEntry}
   > Branch(Condition=IsUsingCompactUI())
   > n1 = Sequence!(execute) -> (then_0, then_1, then_2:Exec, then_3:Exec, then_4:Exec)
   > Set Orientation!(
    execute
    Orientation:EOrientation Enum=Orient_Horizontal
    self:Stack Box Object Reference=Get StackBox_CompactLayout() -> (StackBox_CompactLayout:Stack Box Object Reference)
   ) -> (then, Output_Get:EOrientation Enum=Orient_Horizontal)
   n2 = SlotAsStackBoxSlot!(Widget:Widget Object Reference=Get O_CommunityPost() -> (O_CommunityPost:Horizontal Box Object Reference)) -> (ReturnValue:Stack Box Slot Object Reference)
   Set Padding!(execute=@n1.then_1, Padding:Margin Structure=Make Margin(Left=0, Top=0, Right=0, Bottom=0), self:Stack Box Slot Object Reference=@n2) -> (then, Output_Get:Margin Structure)
   > Set Size!(
    execute
    Size:Slate Child Size Structure=Make Slate Child Size!(Value:float=1, SizeRule:ESlateSizeRule Enum=Fill) -> (SlateChildSize:Slate Child Size Structure) {is=title K2Node_MakeStruct}
    self:Stack Box Slot Object Reference=@n2
   ) -> (then, Output_Get:Slate Child Size Structure)
   > Set VerticalAlignment!(execute, VerticalAlignment:EVerticalAlignment Enum=VAlign_Bottom, self:Stack Box Slot Object Reference=@n2) -> (then, Output_Get:EVerticalAlignment Enum=VAlign_Fill)
   n3 = SlotAsStackBoxSlot(Widget=Get WBP_MOTD_Pip_List())
   Set HorizontalAlignment!(execute=@n1.then_2, HorizontalAlignment:EHorizontalAlignment Enum=HAlign_Right, self:Stack Box Slot Object Reference=@n3) -> (then, Output_Get:EHorizontalAlignment Enum=HAlign_Fill)
   > Set VerticalAlignment(VerticalAlignment=VAlign_Bottom, self=@n3) -> (Output_Get=VAlign_Fill)
   > Set Size(Size=Make Slate Child Size(Value=1, SizeRule=Fill), self=@n3)
   > Set Padding(Padding=Make Margin(Left=0, Top=0, Right=0, Bottom=4), self=@n3)
   n4 = SlotAsHorizontalBoxSlot!(Widget:Widget Object Reference=Get Reaction_Replies() -> (Reaction_Replies:WBP Community Post Reactions Line Object Reference)) -> (ReturnValue:Horizontal Box Slot Object Reference)
   SetVerticalAlignment(execute=@n1.then_3, self:Horizontal Box Slot Object Reference=@n4, InVerticalAlignment=VAlign_Bottom)
   > SetHorizontalAlignment(self:Horizontal Box Slot Object Reference=@n4, InHorizontalAlignment=HAlign_Left)
   > SetSize!(execute, self:Horizontal Box Slot Object Reference=@n4, InSize:Slate Child Size Structure=Make Slate Child Size(Value=1, SizeRule=Automatic)) -> (then)
   n5 = SlotAsHorizontalBoxSlot(Widget=Get Reaction_Like() -> (Reaction_Like:WBP Community Post Reactions Line Object Reference))
   SetVerticalAlignment(execute=@n1.then_4, self=@n5, InVerticalAlignment=VAlign_Bottom)
   > SetHorizontalAlignment(self=@n5, InHorizontalAlignment=HAlign_Left)
   > SetSize(self=@n5, InSize=Make Slate Child Size(Value=1, SizeRule=Automatic))
   > Set Padding(Padding=Make Margin(Left=0, Top=0, Right=16, Bottom=0), self:Horizontal Box Slot Object Reference=@n5)
  ]}
 SetLoadingDisplay {nodes [
   n1 = SetLoadingDisplay!() -> (then, IsLoading:bool=~) {is=title K2Node_FunctionEntry}
   > n2 = GetDynamicMaterial(self=Get Thumbnail())
   > n3 = Branch(Condition=@n1)
   > n4 = Branch(Condition=Get isLoading() -> (isLoading:bool))
   SetScalarParameterValue(execute=@n4.else, self=@n2, ParameterName=Display Loading, Value=1)
   > SetRenderOpacity(self=[Get TeaserTextB(), Get TeaserTextA()], InOpacity=0)
   > Set isLoading(isLoading:bool=true) -> (Output_Get:bool)
   Branch(execute=@n3.else, Condition=Get isLoading())
   > SetScalarParameterValue(self=@n2, ParameterName=Display Loading, Value=0)
   > PlayAnimationForward(InAnimation=Get Anim_ContentLoaded() -> (Anim_ContentLoaded:Widget Animation Object Reference))
   > Set isLoading(isLoading=false)
   n5 = Get IsFirstLoop() -> (IsFirstLoop:bool)
   n6 = "Equal (Object)"(A=GetNextTileInfo(First=NOT Boolean(A=@n5), Peek=false) -> (Param=None).TextBlock, B=Get TeaserTextA())
   > SetRenderOpacity(self=Get TeaserTextA(), InOpacity=SelectFloat(A=1, B=0, bPickA=@n6))
   > SetRenderOpacity(self=Get TeaserTextB(), InOpacity=SelectFloat(A=0, B=1, bPickA=@n6))
  ]}
 VideoEndedEvent {nodes [
   VideoEndedEvent!() -> (then, TextureParam:Name) {is=title K2Node_FunctionEntry}
   > n1 = GetDynamicMaterial(self=Get Thumbnail())
   > SetTextureParameterValue(
    self=@n1
    ParameterName=Get TextureParam() -> (TextureParam=None)
    Value="Get (a copy)"(Array:Array of Texture 2D Object References=Get TeaserImages(), Dimension 1=Get Index() -> (Index=0)) -> (Output:Texture 2D Object Reference)
   )
  ]}
 VideoPlayingEvent {nodes [
   n1 = VideoPlayingEvent!() -> (then, TextureParam:Name, Texture:Media Texture Object Reference, name:Text) {is=title K2Node_FunctionEntry}
   > n2 = GetDynamicMaterial(self=Get Thumbnail())
   > SetTextureParameterValue(self=@n2, ParameterName=@n1, Value=@n1.Texture)
   > Set CurrentTeaser(CurrentTeaser=@n1.name)
  ]}
]
variables [
 ActionInputRow {type=DataTableRowHandle, default {dataTable=&InputActionDataTable, rowName=GenericLeftTrigger}}
 ActivatedOnce {type=boolean}
 Animate Pip {type=boolean, default=true}
 CMS Image {type=CMSImage_C, default=null}
 Count {type=integer}
 CurrentlyUsedTextureParam {type=string, default=None}
 CurrentTeaser {type=string, default=""}
 DebounceUpdateHandle {type=TimerHandle}
 Entry Show Time {type=number, default=5}
 FullScreenMotdActioning {type=boolean}
 HoverSound {type=SoundBase, default=/Game/Sounds/Fort_UI_Sounds/FrontEnd/DefaultSounds/PressedButtons/MSS_UI_Button_Hover_Default}
 ImageA {type=boolean, default=true}
 Index {type=integer}
 IsFirstLoop {type=boolean}
 isLoading {type=boolean}
 LobbyActivated {type=boolean}
 MyMotds {type="array<MotdViewModel>", default []}
 PipsArray {type="array<WBP_MOTD_Pip_Entry_C>", default []}
 PressedSound {type=SoundBase, default=/Game/Sounds/Fort_UI_Sounds/FrontEnd/DefaultSounds/PressedButtons/MSS_UI_Button_Pressed_Default}
 Show Progressbar {type=boolean}
 TeaserImages {type="array<Texture2D>", default []}
 Timer Handle {type=TimerHandle}
]
widgets {
 info {inheritedWidgetCount=0, parentClass=/Script/FortniteUI.FortDynamicMotdTeaserWidget, rootWidgetClass=&Overlay, widgetCount=21}
 widgets [
  Overlay_MOTD {widgetClassPath=&Overlay, bIsVariable=true, widget=Overlay_MOTD}
  SizeBox_67 {
   widgetClassPath=&SizeBox
   parent=Overlay_MOTD
   widget=SizeBox_67
   properties {bOverride_HeightOverride=true, bOverride_WidthOverride=true, heightOverride=200, widthOverride=360}
   slot {class=OverlaySlot, properties {horizontalAlignment=HAlign_Fill, verticalAlignment=VAlign_Fill}}
  }
  Button {
   widgetClassPath=&Button
   parent=SizeBox_67
   widget=Button
   properties {
    clipping=ClipToBounds
    visibility=SelfHitTestInvisible
    backgroundColor=<rgba 1 1 1 0>
    widgetStyle {
     clickedSlateSound {resourceObject=null}
     disabled {
      bIsDynamicallyLoaded=false
      drawAs=NoDrawType
      imageType=NoImage
      mirroring=NoMirror
      resourceName=None
      resourceObject=null
      tiling=NoTile
      imageSize=<0 0>
      margin=<ltrb 0 0 0 0>
      outlineSettings {
       bUseBrushTransparency=false
       roundingType=HalfHeightRadius
       width=0
       color {colorUseRule=UseColor_Specified, specifiedColor=<rgba 0 0 0 0>}
       cornerRadii=<0 0 0 0>
      }
      tintColor {colorUseRule=UseColor_Specified, specifiedColor=<rgba 1 1 1 1>}
      uVRegion {bIsValid=false, max=<0 0>, min=<0 0>}
     }
     disabledForeground {colorUseRule=UseColor_Specified, specifiedColor=<rgba 0.5271 0.5271 0.5271 1>}
     hovered {
      bIsDynamicallyLoaded=false
      drawAs=NoDrawType
      imageType=NoImage
      mirroring=NoMirror
      resourceName=None
      resourceObject=null
      tiling=NoTile
      imageSize=<32 32>
      margin=<ltrb 0 0 0 0>
      outlineSettings {
       bUseBrushTransparency=true
       roundingType=FixedRadius
       width=1
       color {colorUseRule=UseColor_Specified, specifiedColor=<rgba 0.7243 0.7243 0.7243 1>}
       cornerRadii=<4 4 4 4>
      }
      tintColor {colorUseRule=UseColor_Specified, specifiedColor=<rgba 0.7243 0.7243 0.7243 1>}
      uVRegion {bIsValid=false, max=<0 0>, min=<0 0>}
     }
     hoveredForeground {colorUseRule=UseColor_Specified, specifiedColor=<rgba 0 1 1 1>}
     hoveredSlateSound {resourceObject=null}
     normal {
      bIsDynamicallyLoaded=false
      drawAs=NoDrawType
      imageType=NoImage
      mirroring=NoMirror
      resourceName=None
      resourceObject=null
      tiling=NoTile
      imageSize=<32 32>
      margin=<ltrb 0 0 0 0>
      outlineSettings {
       bUseBrushTransparency=true
       roundingType=FixedRadius
       width=1
       color {colorUseRule=UseColor_Specified, specifiedColor=<rgba 0.6951 0.6951 0.6951 1>}
       cornerRadii=<4 4 4 4>
      }
      tintColor {colorUseRule=UseColor_Specified, specifiedColor=<rgba 0.4955 0.4955 0.4955 1>}
      uVRegion {bIsValid=false, max=<0 0>, min=<0 0>}
     }
     normalForeground {colorUseRule=UseColor_Specified, specifiedColor=<rgba 0 1 1 1>}
     normalPadding=<ltrb 0 0 0 0>
     pressed {
      bIsDynamicallyLoaded=false
      drawAs=NoDrawType
      imageType=NoImage
      mirroring=NoMirror
      resourceName=None
      resourceObject=null
      tiling=NoTile
      imageSize=<32 32>
      margin=<ltrb 0 0 0 0>
      outlineSettings {
       bUseBrushTransparency=true
       roundingType=FixedRadius
       width=1
       color {colorUseRule=UseColor_Specified, specifiedColor=<rgba 0.7243 0.7243 0.7243 1>}
       cornerRadii=<4 4 4 4>
      }
      tintColor {colorUseRule=UseColor_Specified, specifiedColor=<rgba 0.3843 0.3843 0.3843 1>}
      uVRegion {bIsValid=false, max=<0 0>, min=<0 0>}
     }
     pressedForeground {colorUseRule=UseColor_Specified, specifiedColor=<rgba 0 1 1 1>}
     pressedPadding=<ltrb 0 0 0 0>
     pressedSlateSound {resourceObject=null}
    }
   }
   slot {class=SizeBoxSlot}
  }
  Overlay_160 {
   widgetClassPath=&Overlay
   parent=Button
   widget=Overlay_160
   properties {clipping=ClipToBoundsAlways}
   slot {class=ButtonSlot, properties {horizontalAlignment=HAlign_Fill, verticalAlignment=VAlign_Fill, padding=<ltrb 0 0 0 0>}}
  }
  ActionButton {
   widgetClassPath=&WBP_UIKit_Button_Generic
   bIsVariable=true
   parent=Overlay_160
   widget=ActionButton
   properties {style=/FortUILibrary/Foundation/Buttons/ButtonStyles/ButtonStyle-Empty.ButtonStyle-Empty_C, text="", colorAndOpacity=<rgba 1 1 1 0>}
   slot {class=OverlaySlot, properties {horizontalAlignment=HAlign_Fill, verticalAlignment=VAlign_Fill}}
  }
  Thumbnail {
   widgetClassPath=&Image
   bIsVariable=true
   parent=Overlay_160
   widget=Thumbnail
   properties {
    visibility=HitTestInvisible
    brush {
     bIsDynamicallyLoaded=false
     drawAs=Image
     imageType=NoImage
     mirroring=NoMirror
     resourceName=None
     resourceObject=/Motd/Widgets/Materials/MI_MOTD_Tile
     tiling=NoTile
     imageSize=<1024 512>
     margin=<ltrb 0 0 0 0>
     outlineSettings {
      bUseBrushTransparency=false
      roundingType=HalfHeightRadius
      width=0
      color {colorUseRule=UseColor_Specified, specifiedColor=<rgba 0 0 0 0>}
      cornerRadii=<0 0 0 0>
     }
     tintColor {colorUseRule=UseColor_Specified, specifiedColor=<rgba 1 1 1 1>}
     uVRegion {bIsValid=false, max=<0 0>, min=<0 0>}
    }
   }
   slot {class=OverlaySlot, properties {horizontalAlignment=HAlign_Fill, verticalAlignment=VAlign_Fill}}
  }
  GridPanel_0 {
   widgetClassPath=&GridPanel
   parent=Overlay_160
   widget=GridPanel_0
   properties {columnFill [1], rowFill [0,1]}
   slot {class=OverlaySlot, properties {horizontalAlignment=HAlign_Fill, verticalAlignment=VAlign_Fill, padding=<ltrb 16 16 16 16>}}
  }
  CommonVisibilityWidget {
   widgetClassPath=&CommonVisibilityWidgetLegacy
   parent=GridPanel_0
   widget=CommonVisibilityWidget
   properties {bShowForMouseAndKeyboard=false, bShowForTouch=false, padding=<ltrb 0 0 0 0>}
   slot {class=GridSlot, properties {horizontalAlignment=HAlign_Left, layer=1, verticalAlignment=VAlign_Top}}
  }
  InputActionWidget {
   widgetClassPath=&CommonActionWidget
   bIsVariable=true
   parent=CommonVisibilityWidget
   widget=InputActionWidget
   properties {inputActions [{dataTable=&InputActionDataTable, rowName=GenericLeftTrigger}]}
   slot {class=BorderSlot, properties {padding=<ltrb 0 0 0 0>}}
  }
  EventHighlight {
   widgetClassPath=/FortUILibrary/UIKit/Highlight/WBP_UIKit_Highlight
   bIsVariable=true
   parent=GridPanel_0
   widget=EventHighlight
   properties {highlightSize=Compact, highlightText=live now, highlightType=LiveAlt}
   slot {class=GridSlot, properties {horizontalAlignment=HAlign_Right, verticalAlignment=VAlign_Top}}
  }
  VerticalBox_0 {
   widgetClassPath=&VerticalBox
   parent=GridPanel_0
   widget=VerticalBox_0
   slot {class=GridSlot, properties {row=1, rowSpan=1, verticalAlignment=VAlign_Bottom}}
  }
  Text_Interested {
   widgetClassPath=&CommonTextBlock
   bIsVariable=true
   parent=VerticalBox_0
   widget=Text_Interested
   properties {
    mobileFontSizeMultiplier=1.2
    text="## interested"
    textTransformPolicy=ToUpper
    visibility=Collapsed
    font {
     bForceMonospaced=false
     bMaterialIsStencil=false
     fontMaterial=null
     fontObject=&HeadingNow
     letterSpacing=10
     monospacedWidth=1
     size=12.75
     skewAmount=0
     typefaceFontName=75 Medium
     outlineSettings {
      bApplyOutlineToDropShadows=false
      bMiteredCorners=false
      bSeparateFillAlpha=false
      outlineMaterial=null
      outlineSize=1
      outlineColor=<rgba 0 0 0 0.5>
     }
    }
    shadowOffset=<0 0>
   }
   slot {class=VerticalBoxSlot, properties {padding=<ltrb 0 0 0 10>}}
  }
  O_Default {
   widgetClassPath=&Overlay
   parent=VerticalBox_0
   widget=O_Default
   slot {class=VerticalBoxSlot, properties {size {sizeRule=Fill, value=1}}}
  }
  TeaserTextB {
   widgetClassPath=&CommonRichTextBlock
   bIsVariable=true
   parent=O_Default
   widget=TeaserTextB
   properties {
    applyLineHeightToBottomLine=false
    autoWrapText=true
    bOverrideDefaultStyle=true
    clipping=ClipToBounds
    lineHeightPercentage=0.8
    mobileTextBlockScale=1.1
    renderOpacity=0
    text=$Teaser Label B
    textOverflowPolicy=MultilineEllipsis
    textTransformPolicy=ToUpper
    defaultTextStyleOverride {
     overflowPolicy=MultilineEllipsis
     transformPolicy=None
     colorAndOpacity {colorUseRule=UseColor_Specified, specifiedColor=<rgba 1 1 1 1>}
     font {
      bForceMonospaced=false
      bMaterialIsStencil=false
      fontMaterial=null
      fontObject=&HeadingNow
      letterSpacing=10
      monospacedWidth=1
      size=15.75
      skewAmount=0
      typefaceFontName=86 Bold Italic
      outlineSettings {
       bApplyOutlineToDropShadows=false
       bMiteredCorners=false
       bSeparateFillAlpha=false
       outlineMaterial=null
       outlineSize=0
       outlineColor=<rgba 0 0 0 1>
      }
     }
     highlightColor {colorUseRule=UseColor_Specified, specifiedColor=<rgba 0 0 0 1>}
     highlightShape {
      bIsDynamicallyLoaded=false
      drawAs=Image
      imageType=NoImage
      mirroring=NoMirror
      resourceName=None
      resourceObject=null
      tiling=NoTile
      imageSize=<32 32>
      margin=<ltrb 0 0 0 0>
      outlineSettings {
       bUseBrushTransparency=false
       roundingType=HalfHeightRadius
       width=0
       color {colorUseRule=UseColor_Specified, specifiedColor=<rgba 0 0 0 0>}
       cornerRadii=<0 0 0 0>
      }
      tintColor {colorUseRule=UseColor_Specified, specifiedColor=<rgba 1 1 1 1>}
      uVRegion {bIsValid=false, max=<0 0>, min=<0 0>}
     }
     selectedBackgroundColor {colorUseRule=UseColor_Foreground, specifiedColor=<rgba 1 0 1 1>}
     shadowColorAndOpacity=<rgba 0 0 0 1>
     shadowOffset=<0 0>
     strikeBrush {
      bIsDynamicallyLoaded=false
      drawAs=Image
      imageType=NoImage
      mirroring=NoMirror
      resourceName=None
      resourceObject=null
      tiling=NoTile
      imageSize=<32 32>
      margin=<ltrb 0 0 0 0>
      outlineSettings {
       bUseBrushTransparency=false
       roundingType=HalfHeightRadius
       width=0
       color {colorUseRule=UseColor_Specified, specifiedColor=<rgba 0 0 0 0>}
       cornerRadii=<0 0 0 0>
      }
      tintColor {colorUseRule=UseColor_Specified, specifiedColor=<rgba 1 1 1 1>}
      uVRegion {bIsValid=false, max=<0 0>, min=<0 0>}
     }
     underlineBrush {
      bIsDynamicallyLoaded=false
      drawAs=Image
      imageType=NoImage
      mirroring=NoMirror
      resourceName=None
      resourceObject=null
      tiling=NoTile
      imageSize=<32 32>
      margin=<ltrb 0 0 0 0>
      outlineSettings {
       bUseBrushTransparency=false
       roundingType=HalfHeightRadius
       width=0
       color {colorUseRule=UseColor_Specified, specifiedColor=<rgba 0 0 0 0>}
       cornerRadii=<0 0 0 0>
      }
      tintColor {colorUseRule=UseColor_Specified, specifiedColor=<rgba 1 1 1 1>}
      uVRegion {bIsValid=false, max=<0 0>, min=<0 0>}
     }
    }
   }
   slot {class=OverlaySlot, properties {horizontalAlignment=HAlign_Fill, verticalAlignment=VAlign_Bottom}}
  }
  TeaserTextA {
   widgetClassPath=&CommonRichTextBlock
   bIsVariable=true
   parent=O_Default
   widget=TeaserTextA
   properties {
    applyLineHeightToBottomLine=false
    autoWrapText=true
    bOverrideDefaultStyle=true
    clipping=ClipToBounds
    lineHeightPercentage=0.8
    mobileTextBlockScale=1.1
    renderOpacity=0
    text=$Teaser Label A
    textOverflowPolicy=MultilineEllipsis
    textTransformPolicy=ToUpper
    defaultTextStyleOverride {
     overflowPolicy=MultilineEllipsis
     transformPolicy=None
     colorAndOpacity {colorUseRule=UseColor_Specified, specifiedColor=<rgba 1 1 1 1>}
     font {
      bForceMonospaced=false
      bMaterialIsStencil=false
      fontMaterial=null
      fontObject=&HeadingNow
      letterSpacing=10
      monospacedWidth=1
      size=15.75
      skewAmount=0
      typefaceFontName=86 Bold Italic
      outlineSettings {
       bApplyOutlineToDropShadows=false
       bMiteredCorners=false
       bSeparateFillAlpha=false
       outlineMaterial=null
       outlineSize=0
       outlineColor=<rgba 0 0 0 1>
      }
     }
     highlightColor {colorUseRule=UseColor_Specified, specifiedColor=<rgba 0 0 0 1>}
     highlightShape {
      bIsDynamicallyLoaded=false
      drawAs=Image
      imageType=NoImage
      mirroring=NoMirror
      resourceName=None
      resourceObject=null
      tiling=NoTile
      imageSize=<32 32>
      margin=<ltrb 0 0 0 0>
      outlineSettings {
       bUseBrushTransparency=false
       roundingType=HalfHeightRadius
       width=0
       color {colorUseRule=UseColor_Specified, specifiedColor=<rgba 0 0 0 0>}
       cornerRadii=<0 0 0 0>
      }
      tintColor {colorUseRule=UseColor_Specified, specifiedColor=<rgba 1 1 1 1>}
      uVRegion {bIsValid=false, max=<0 0>, min=<0 0>}
     }
     selectedBackgroundColor {colorUseRule=UseColor_Foreground, specifiedColor=<rgba 1 0 1 1>}
     shadowColorAndOpacity=<rgba 0 0 0 1>
     shadowOffset=<0 0>
     strikeBrush {
      bIsDynamicallyLoaded=false
      drawAs=Image
      imageType=NoImage
      mirroring=NoMirror
      resourceName=None
      resourceObject=null
      tiling=NoTile
      imageSize=<32 32>
      margin=<ltrb 0 0 0 0>
      outlineSettings {
       bUseBrushTransparency=false
       roundingType=HalfHeightRadius
       width=0
       color {colorUseRule=UseColor_Specified, specifiedColor=<rgba 0 0 0 0>}
       cornerRadii=<0 0 0 0>
      }
      tintColor {colorUseRule=UseColor_Specified, specifiedColor=<rgba 1 1 1 1>}
      uVRegion {bIsValid=false, max=<0 0>, min=<0 0>}
     }
     underlineBrush {
      bIsDynamicallyLoaded=false
      drawAs=Image
      imageType=NoImage
      mirroring=NoMirror
      resourceName=None
      resourceObject=null
      tiling=NoTile
      imageSize=<32 32>
      margin=<ltrb 0 0 0 0>
      outlineSettings {
       bUseBrushTransparency=false
       roundingType=HalfHeightRadius
       width=0
       color {colorUseRule=UseColor_Specified, specifiedColor=<rgba 0 0 0 0>}
       cornerRadii=<0 0 0 0>
      }
      tintColor {colorUseRule=UseColor_Specified, specifiedColor=<rgba 1 1 1 1>}
      uVRegion {bIsValid=false, max=<0 0>, min=<0 0>}
     }
    }
   }
   slot {class=OverlaySlot, properties {horizontalAlignment=HAlign_Fill, verticalAlignment=VAlign_Bottom}}
  }
  StackBox_CompactLayout {
   widgetClassPath=&StackBox
   bIsVariable=true
   parent=VerticalBox_0
   widget=StackBox_CompactLayout
   properties {orientation=Orient_Vertical}
   slot {class=VerticalBoxSlot}
  }
  O_CommunityPost {
   widgetClassPath=&HorizontalBox
   bIsVariable=true
   parent=StackBox_CompactLayout
   widget=O_CommunityPost
   properties {bOverride_Cursor=true, visibility=Collapsed}
   slot {class=StackBoxSlot}
  }
  Reaction_Like {
   widgetClassPath=/FortniteCoreUI/UI/Community/WBP_CommunityPost_ReactionsLine
   bIsVariable=true
   parent=O_CommunityPost
   widget=Reaction_Like
   properties {
    brush Visibility=SelfHitTestInvisible
    emoji Visibility=Collapsed
    material=/FortniteCoreUI/UI/Community/Art/MaterialInstances/MI_UI_Communities_Reaction_Emoji
    in Color and Opacity=<rgba 1 1 1 1>
   }
   slot {
    class=HorizontalBoxSlot
    properties {horizontalAlignment=HAlign_Left, verticalAlignment=VAlign_Bottom, padding=<ltrb 0 5 0 0>, size {sizeRule=Fill, value=1}}
   }
  }
  Reaction_Replies {
   widgetClassPath=&WBP_CommunityPost_ReactionsLine
   bIsVariable=true
   parent=O_CommunityPost
   widget=Reaction_Replies
   properties {emoji="💬", emoji Visibility=SelfHitTestInvisible}
   slot {
    class=HorizontalBoxSlot
    properties {horizontalAlignment=HAlign_Right, verticalAlignment=VAlign_Bottom, padding=<ltrb 0 5 0 0>, size {sizeRule=Fill, value=1}}
   }
  }
  WBP_MOTD_Pip_List {
   widgetClassPath=&/WBP_MOTD_Pip_List
   bIsVariable=true
   parent=StackBox_CompactLayout
   widget=WBP_MOTD_Pip_List
   slot {class=StackBoxSlot, properties {padding=<ltrb 0 8 0 0>}}
  }
  Block_Outline {
   widgetClassPath=/FortUILibrary/UIKit/Buttons/Blocks/WBP_UIKit_Block_Outline
   bIsVariable=true
   parent=Overlay_MOTD
   widget=Block_Outline
   properties {bottomLeftCornerRadius=28, bottomRightCornerRadius=28, radiusInPixels=true, topLeftCornerRadius=28, topRightCornerRadius=28}
   slot {class=OverlaySlot, properties {horizontalAlignment=HAlign_Fill, verticalAlignment=VAlign_Fill}}
  }
 ]
}
